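# Corrective Control Rig for the F_MED_Moss_Bomber player cosmetic body, nudging foot support and shoulder leaf bones as the pose changes.
# Spherical pose readers on each foot and on spine_05 output weights that scale Modify Transforms offsets.
# A sequence runs the feet first, then both shoulder leaf joints from the shared spine reader.
# Debug drawing is enabled on all readers, and the leading parent constraint has no bones assigned.
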ControlRigBlueprint /BRCosmetics/Characters/Player/Female/Medium/Bodies/F_MED_Moss_Bomber/Meshes/F_MED_Moss_Bomber_CtrlRig
graphs [
 RigVMFunctionLibrary {links [], nodes []}
 RigVMModel {nodes [
   RigUnit_BeginExecution = &RigUnit_BeginExecution()
   n1 = &RigUnit_ParentConstraint(
    AdvancedSettings=(InterpolationType=Average,RotationOrderForFilter=XZY)
    Child=(Type=Bone,Name="None")
    Parents=((Item=(Type=Bone,Name="None"),Weight=1))
    Weight=1
    ExecutePin=@RigUnit_BeginExecution.ExecutePin
   )
   RigVMFunction_Sequence = &RigVMFunction_Sequence(ExecuteContext=@n1.ExecutePin)
   n2 = &RigUnit_SphericalPoseReader(
    ActiveRegionScaleFactors=(PositiveWidth=1,NegativeWidth=1,PositiveHeight=1,NegativeHeight=1)
    ActiveRegionSize=0.1
    Debug=(bDrawDebug=true,bDraw2D=False,bDrawLocalAxes=False,DebugScale=25,DebugSegments=20,DebugThickness=0.25)
    DriverAxis=(X=0,Y=1,Z=0)
    DriverItem=(Type=Bone,Name="foot_r")
    FalloffRegionScaleFactors=(PositiveWidth=1,NegativeWidth=1,PositiveHeight=1,NegativeHeight=1.2)
    FalloffSize=0.2
    FlipHeightScaling=False
    FlipWidthScaling=False
    OptionalParentItem=(Type=Bone,Name="None")
    RotationOffset=(X=145,Y=-90,Z=0)
    ExecutePin=@RigVMFunction_Sequence.A
   )
   > &RigUnit_ModifyTransforms(
    ItemToModify=((Item=(Type=Bone,Name="foot_support_r"),Transform=(Rotation=(X=0,Y=0,Z=0,W=1),Translation=(X=0,Y=1.5608,Z=7.4358),Scale3D=(X=1,Y=1,Z=1))))
    WeightMinimum=0
    Weight=@n2.OutputParam
   )
   > n3 = &RigUnit_SphericalPoseReader(
    ActiveRegionScaleFactors=(PositiveWidth=1,NegativeWidth=1,PositiveHeight=1,NegativeHeight=1)
    ActiveRegionSize=0.1
    Debug=(bDrawDebug=true,bDraw2D=False,bDrawLocalAxes=False,DebugScale=25,DebugSegments=20,DebugThickness=0.25)
    DriverAxis=(X=0,Y=-1,Z=0)
    DriverItem=(Type=Bone,Name="foot_l")
    FalloffRegionScaleFactors=(PositiveWidth=1,NegativeWidth=1,PositiveHeight=1,NegativeHeight=1.2)
    FalloffSize=0.2
    FlipHeightScaling=False
    FlipWidthScaling=False
    OptionalParentItem=(Type=Bone,Name="None")
    RotationOffset=(X=-30,Y=-75,Z=0)
   )
   > &RigUnit_ModifyTransforms(
    ItemToModify=((Item=(Type=Bone,Name="foot_support_l"),Transform=(Rotation=(X=0,Y=0,Z=0,W=1),Translation=(X=0,Y=-2.5728,Z=-7),Scale3D=(X=1,Y=1,Z=1))))
    WeightMinimum=0
    Weight=@n3.OutputParam
   )
   n4 = &RigUnit_SphericalPoseReader(
    ActiveRegionScaleFactors=(PositiveWidth=1,NegativeWidth=1,PositiveHeight=1,NegativeHeight=1)
    ActiveRegionSize=0.1
    Debug=(bDrawDebug=True,bDraw2D=False,bDrawLocalAxes=False,DebugScale=25,DebugSegments=20,DebugThickness=0.25)
    DriverAxis=(X=0,Y=-1,Z=0)
    DriverItem=(Type=Bone,Name="spine_05")
    FalloffRegionScaleFactors=(PositiveWidth=1,NegativeWidth=1,PositiveHeight=1,NegativeHeight=1)
    FalloffSize=0.2
    FlipHeightScaling=False
    FlipWidthScaling=False
    OptionalParentItem=(Type=Bone,Name="None")
    RotationOffset=(X=-50,Y=0,Z=0)
    ExecutePin=@RigVMFunction_Sequence.B
   )
   > &RigUnit_ModifyTransforms(
    ItemToModify=((Item=(Type=Bone,Name="dyn_R_shoulderLeaves_jnt_1"),Transform=(Rotation=(X=0,Y=0,Z=0,W=1),Translation=(X=-3.1556,Y=-3.8121,Z=8.4115),Scale3D=(X=1,Y=1,Z=1))))
    WeightMinimum=0
    Weight=@n4.OutputParam
   )
   > &RigUnit_ModifyTransforms(
    ItemToModify=((Item=(Type=Bone,Name="dyn_L_shoulderLeaves_jnt_1"),Transform=(Rotation=(X=0,Y=0,Z=0,W=-1),Translation=(X=0,Y=3.2532,Z=3.8433),Scale3D=(X=1,Y=1,Z=1))))
    WeightMinimum=0
    Weight=@n4.OutputParam
   )
  ]}
]
hierarchy [
 root {loc=<0 0 0>, rot=<0 0 0 1>}
 attach {parent=root, loc=<0 0 0>, rot=<0 0 0 1>}
 pelvis {parent=root, loc=<0 -0.5301 98.0571>, rot=<0.0127 0.707 -0.0127 -0.707>}
 spine_01 {parent=pelvis, loc=<6.1208 -0.9328 0>, rot=<0 0 0.034 -0.9994>}
 spine_02 {parent=spine_01, loc=<6.5996 -0.2095 0>, rot=<0 0 0.0238 -0.9997>}
 spine_03 {parent=spine_02, loc=<6.4914 -0.1237 0>, rot=<0 0 0.0149 -0.9999>}
 spine_04 {parent=spine_03, loc=<6.9598 0.0794 0>, rot=<0 0 -0.0132 -0.9999>}
 spine_05 {parent=spine_04, loc=<6.9987 0 0>, rot=<0 0 -0.0015 -1>}
 clavicle_l {parent=spine_05, loc=<5.4831 -0.1053 -3.2946>, rot=<0.0465 -0.7289 -0.009 -0.6829>}
 upperarm_l {parent=clavicle_l, loc=<9.8658 -1.2136 -0.4105>, rot=<0.0164 -0.3245 -0.0269 -0.9454>}
 lowerarm_l {parent=upperarm_l, loc=<26.465 -0.8345 0.259>, rot=<0.0036 0.0161 -0.2298 -0.9731>}
 hand_l {parent=lowerarm_l, loc=<22.7355 -0.3203 -0.6105>, rot=<-0.6784 -0.0223 0.0735 -0.7307>}
 index_metacarpal_l {parent=hand_l, loc=<1.6535 0.0183 -2.96>, rot=<0.0158 -0.0124 -0.0173 -0.9996>}
 index_01_l {parent=index_metacarpal_l, loc=<8.3008 0.0357 -0.0524>, rot=<-0.0044 0.0485 0.0924 -0.9945>}
 index_02_l {parent=index_01_l, loc=<3.3587 -0.3004 0.0452>, rot=<-0.002 0.0027 0.3225 -0.9466>}
 index_03_l {parent=index_02_l, loc=<2.5578 -0.123 -0.0252>, rot=<-0.002 0.0067 0.2558 -0.9667>}
 middle_metacarpal_l {parent=hand_l, loc=<2.1856 -0.1725 -0.5984>, rot=<0.006 0.0168 -0.0329 -0.9993>}
 middle_01_l {parent=middle_metacarpal_l, loc=<7.3275 -0.2447 0.0129>, rot=<-0.0208 0.0042 0.0858 -0.9961>}
 middle_02_l {parent=middle_01_l, loc=<4.0123 0.1401 0.1175>, rot=<-0.0218 0.0258 0.3512 -0.9357>}
 middle_03_l {parent=middle_02_l, loc=<2.6077 0.0534 0.1373>, rot=<0.0082 0.0188 0.1708 -0.9851>}
 pinky_metacarpal_l {parent=hand_l, loc=<1.7616 -0.9753 3.3814>, rot=<-0.0397 0.0814 -0.069 -0.9935>}
 pinky_01_l {parent=pinky_metacarpal_l, loc=<7.1709 -0.4831 -0.1094>, rot=<-0.0654 -0.0035 0.2242 -0.9723>}
 pinky_02_l {parent=pinky_01_l, loc=<2.8274 0.0893 -0.1043>, rot=<0.0037 0.0011 0.2854 -0.9584>}
 pinky_03_l {parent=pinky_02_l, loc=<2.2046 0.0307 -0.0078>, rot=<-0.0002 0.0098 0.1445 -0.9895>}
 ring_metacarpal_l {parent=hand_l, loc=<2.2074 -0.3625 1.3638>, rot=<-0.0284 0.0722 -0.0458 -0.9959>}
 ring_01_l {parent=ring_metacarpal_l, loc=<7.0295 -0.6197 -0.1446>, rot=<-0.0489 -0.0252 0.1892 -0.9804>}
 ring_02_l {parent=ring_01_l, loc=<3.6174 0.4647 -0.1027>, rot=<0 0 0.2806 -0.9598>}
 ring_03_l {parent=ring_02_l, loc=<2.4152 0.114 0.0199>, rot=<0 0 0.1189 -0.9929>}
 thumb_01_l {parent=hand_l, loc=<1.3118 -0.3191 -2.459>, rot=<0.6546 -0.4042 0.0664 -0.6354>}
 thumb_02_l {parent=thumb_01_l, loc=<5.9336 -0.0448 0.0401>, rot=<-0.0037 0.0213 0.298 -0.9543>}
 thumb_03_l {parent=thumb_02_l, loc=<2.2223 0.3682 -0.0907>, rot=<0.0118 -0.013 0.0066 -0.9998>}
 weapon_l {parent=hand_l, loc=<-0.1974 -3.4111 0.3811>, rot=<0 0 0 1>}
 lowerarm_twist_01_l {parent=lowerarm_l, loc=<18.316 0 0>, rot=<0 0 0 1>}
 lowerarm_twist_02_l {parent=lowerarm_l, loc=<10.6707 0 0>, rot=<0 0 0 1>}
 upperarm_twist_01_l {parent=upperarm_l, loc=<7.2135 -0.1134 0>, rot=<0 0 0 1>}
 dyn_L_leaf_jnt_12 {parent=upperarm_twist_01_l, loc=<-2.1102 -0.4531 4.7412>, rot=<-0.0065 0.3563 0.0039 -0.9343>}
 dyn_L_leafDynamic_jnt_10 {parent=dyn_L_leaf_jnt_12, loc=<1.3314 -0.4717 -0.3015>, rot=<-0.0492 -0.2496 0.1871 -0.9488>}
 dyn_L_leaf_jnt_15 {parent=upperarm_twist_01_l, loc=<-3.6828 -2.1178 3.7433>, rot=<-0.0065 0.3563 0.0039 -0.9343>}
 dyn_L_leaf_jnt_8 {parent=upperarm_twist_01_l, loc=<-4.3696 0.1935 4.787>, rot=<-0.0065 0.3563 0.0039 -0.9343>}
 dyn_L_leaf_jnt_7 {parent=upperarm_twist_01_l, loc=<-3.4485 2.4933 4.4841>, rot=<-0.0065 0.3563 0.0039 -0.9343>}
 dyn_L_leaf_jnt_10 {parent=upperarm_twist_01_l, loc=<-7.2268 -1.7282 4.1212>, rot=<-0.0065 0.3563 0.0039 -0.9343>}
 dyn_L_leafDynamic_jnt_5 {parent=dyn_L_leaf_jnt_10, loc=<1.3992 -0.158 0.3965>, rot=<-0.3019 0.0454 0.0144 -0.9521>}
 dyn_L_leaf_jnt_6 {parent=upperarm_twist_01_l, loc=<-3.2665 4.932 1.6888>, rot=<-0.0065 0.3563 0.0039 -0.9343>}
 dyn_L_leaf_jnt_3 {parent=upperarm_twist_01_l, loc=<-6.2925 4.0912 3.314>, rot=<-0.0065 0.3563 0.0039 -0.9343>}
 dyn_L_leaf_jnt_2 {parent=upperarm_twist_01_l, loc=<-7.3178 3.1775 3.7248>, rot=<-0.0065 0.3563 0.0039 -0.9343>}
 dyn_L_leafDynamic_jnt_4 {parent=dyn_L_leaf_jnt_2, loc=<0.6206 0.2904 0.4711>, rot=<0.1704 0.1273 -0.0233 -0.9768>}
 dyn_L_leaf_jnt_11 {parent=upperarm_twist_01_l, loc=<-6.5328 -4.3264 1.8448>, rot=<-0.0065 0.3563 0.0039 -0.9343>}
 dyn_L_leafDynamic_jnt_9 {parent=dyn_L_leaf_jnt_11, loc=<0.9197 -0.5442 -0.3068>, rot=<-0.0897 -0.2847 0.2869 -0.9103>}
 upperarm_twist_02_l {parent=upperarm_l, loc=<19.7998 -0.1134 0>, rot=<0 0 0 1>}
 dyn_L_leafDynamic_jnt_7 {parent=upperarm_twist_02_l, loc=<-3.8244 -6.0031 -1.2801>, rot=<-0.0146 -0.2499 -0.0094 -0.9681>}
 dyn_L_leafDynamic_jnt_8 {parent=upperarm_twist_02_l, loc=<-0.0604 -4.0226 4.0676>, rot=<-0.2629 0.3469 -0.2328 -0.8697>}
 dyn_L_leaf_jnt_16 {parent=upperarm_l, loc=<-2.2636 1.6811 3.6283>, rot=<-0.0065 0.3563 0.0039 -0.9343>}
 dyn_L_leafDynamic_jnt_2 {parent=dyn_L_leaf_jnt_16, loc=<0.5649 -0.129 0.8199>, rot=<-0.0436 0.3752 0.1068 -0.9198>}
 dyn_L_leafDynamic_jnt_3 {parent=dyn_L_leafDynamic_jnt_2, loc=<1.8665 -0.053 -0.0054>, rot=<0.0446 -0.3649 0.1014 -0.9244>}
 dyn_L_leaf_jnt_14 {parent=upperarm_l, loc=<-0.9127 -4.5052 1.9212>, rot=<-0.0065 0.3563 0.0039 -0.9343>}
 dyn_L_leaf_jnt_13 {parent=upperarm_l, loc=<-0.4682 -5.0383 0.2214>, rot=<-0.0065 0.3563 0.0039 -0.9343>}
 dyn_L_leaf_jnt_1 {parent=clavicle_l, loc=<6.8858 -2.6128 4.388>, rot=<0.0008 0.0336 -0.0252 -0.9991>}
 dyn_L_leafDynamic_jnt_1 {parent=dyn_L_leaf_jnt_1, loc=<0 0 0>, rot=<0.1425 0.1734 0.5534 -0.8021>}
 dyn_L_leaf_jnt_4 {parent=clavicle_l, loc=<8.1964 3.804 0.9317>, rot=<0.0008 0.0336 -0.0252 -0.9991>}
 dyn_L_leafDynamic_jnt_6 {parent=dyn_L_leaf_jnt_4, loc=<0 0 0>, rot=<-0.0257 0.0381 -0.558 -0.8286>}
 dyn_L_leaf_jnt_9 {parent=clavicle_l, loc=<6.3337 -6.6094 1.3032>, rot=<0.0008 0.0336 -0.0252 -0.9991>}
 clavicle_r {parent=spine_05, loc=<5.4831 -0.1053 3.2946>, rot=<0.7289 0.0465 -0.6829 0.009>}
 upperarm_r {parent=clavicle_r, loc=<-9.8658 1.2136 0.4104>, rot=<0.0164 -0.3245 -0.0269 -0.9454>}
 lowerarm_r {parent=upperarm_r, loc=<-26.4651 0.8345 -0.259>, rot=<0.0036 0.0161 -0.2298 -0.9731>}
 hand_r {parent=lowerarm_r, loc=<-22.7356 0.3203 0.6106>, rot=<-0.6784 -0.0223 0.0735 -0.7307>}
 index_metacarpal_r {parent=hand_r, loc=<-1.6534 -0.0181 2.9601>, rot=<0.0158 -0.0124 -0.0173 -0.9996>}
 index_01_r {parent=index_metacarpal_r, loc=<-8.3008 -0.0357 0.0524>, rot=<-0.0044 0.0485 0.0924 -0.9945>}
 index_02_r {parent=index_01_r, loc=<-3.3589 0.3003 -0.0453>, rot=<-0.002 0.0027 0.3225 -0.9466>}
 index_03_r {parent=index_02_r, loc=<-2.5578 0.123 0.0252>, rot=<-0.002 0.0067 0.2558 -0.9667>}
 middle_metacarpal_r {parent=hand_r, loc=<-2.1856 0.1726 0.5984>, rot=<0.006 0.0168 -0.0329 -0.9993>}
 middle_01_r {parent=middle_metacarpal_r, loc=<-7.3273 0.2447 -0.0129>, rot=<-0.0208 0.0042 0.0858 -0.9961>}
 middle_02_r {parent=middle_01_r, loc=<-4.0124 -0.14 -0.1175>, rot=<-0.0218 0.0258 0.3512 -0.9357>}
 middle_03_r {parent=middle_02_r, loc=<-2.6077 -0.0533 -0.1373>, rot=<0.0082 0.0188 0.1708 -0.9851>}
 pinky_metacarpal_r {parent=hand_r, loc=<-1.7617 0.9753 -3.3814>, rot=<-0.0397 0.0814 -0.069 -0.9935>}
 pinky_01_r {parent=pinky_metacarpal_r, loc=<-7.1707 0.4832 0.1094>, rot=<-0.0654 -0.0035 0.2242 -0.9723>}
 pinky_02_r {parent=pinky_01_r, loc=<-2.8274 -0.0892 0.1043>, rot=<0.0037 0.0011 0.2854 -0.9584>}
 pinky_03_r {parent=pinky_02_r, loc=<-2.2045 -0.031 0.0078>, rot=<-0.0002 0.0098 0.1445 -0.9895>}
 ring_metacarpal_r {parent=hand_r, loc=<-2.2073 0.3627 -1.3638>, rot=<-0.0284 0.0722 -0.0458 -0.9959>}
 ring_01_r {parent=ring_metacarpal_r, loc=<-7.0297 0.6197 0.1446>, rot=<-0.0489 -0.0252 0.1892 -0.9804>}
 ring_02_r {parent=ring_01_r, loc=<-3.6171 -0.4648 0.1027>, rot=<0 0 0.2806 -0.9598>}
 ring_03_r {parent=ring_02_r, loc=<-2.4153 -0.1141 -0.0199>, rot=<0 0 0.1189 -0.9929>}
 thumb_01_r {parent=hand_r, loc=<-1.3119 0.3192 2.459>, rot=<0.6546 -0.4042 0.0664 -0.6354>}
 thumb_02_r {parent=thumb_01_r, loc=<-5.9335 0.0447 -0.0401>, rot=<-0.0037 0.0213 0.298 -0.9543>}
 thumb_03_r {parent=thumb_02_r, loc=<-2.2223 -0.3682 0.0907>, rot=<0.0118 -0.013 0.0066 -0.9998>}
 weapon_r {parent=hand_r, loc=<0.1975 3.4112 -0.3811>, rot=<0 0 0 -1>}
 lowerarm_twist_01_r {parent=lowerarm_r, loc=<-18.3161 0 0>, rot=<0 0 0 1>}
 lowerarm_twist_02_r {parent=lowerarm_r, loc=<-10.6707 0 0>, rot=<0 0 0 1>}
 upperarm_twist_01_r {parent=upperarm_r, loc=<-7.2135 0.1134 0>, rot=<0 0 0 1>}
 dyn_R_leaf_jnt_12 {parent=upperarm_twist_01_r, loc=<-0.1601 -2.7816 -2.5617>, rot=<0.9343 -0.0039 0.3563 -0.0065>}
 dyn_R_leaf_jnt_14 {parent=upperarm_twist_01_r, loc=<-4.2841 2.0429 -3.3113>, rot=<0.9343 -0.0039 0.3563 -0.0065>}
 dyn_R_leaf_jnt_17 {parent=upperarm_twist_01_r, loc=<1.6139 4.0019 -1.918>, rot=<0.9343 -0.0039 0.3563 -0.0065>}
 dyn_R_leaf_jnt_15 {parent=upperarm_twist_01_r, loc=<0.746 1.4149 -4.27>, rot=<0.9343 -0.0039 0.3563 -0.0065>}
 dyn_R_leaf_jnt_13 {parent=upperarm_twist_01_r, loc=<1.6911 -4.7623 -1.7171>, rot=<0.9343 -0.0039 0.3563 -0.0065>}
 dyn_R_leaf_jnt_16 {parent=upperarm_twist_01_r, loc=<2.3678 3.404 -2.6839>, rot=<0.9343 -0.0039 0.3563 -0.0065>}
 dyn_R_leaf_jnt_9 {parent=upperarm_twist_01_r, loc=<2.9574 -2.5554 -4.482>, rot=<0.9343 -0.0039 0.3563 -0.0065>}
 dyn_R_leaf_jnt_6 {parent=upperarm_twist_01_r, loc=<5.7807 -5.5963 -1.5944>, rot=<0.9343 -0.0039 0.3563 -0.0065>}
 dyn_R_leaf_jnt_7 {parent=upperarm_twist_01_r, loc=<4.8179 -3.8746 -3.4398>, rot=<0.9343 -0.0039 0.3563 -0.0065>}
 dyn_R_leaf_jnt_10 {parent=upperarm_twist_01_r, loc=<4.7341 -0.7948 -4.6934>, rot=<0.9343 -0.0039 0.3563 -0.0065>}
 dyn_R_leaf_jnt_11 {parent=upperarm_twist_01_r, loc=<6.6396 1.1099 -4.584>, rot=<0.9343 -0.0039 0.3563 -0.0065>}
 upperarm_twist_02_r {parent=upperarm_r, loc=<-19.7998 0.1134 0>, rot=<0 0 0 1>}
 dyn_knot_kinematic_jnt {parent=upperarm_twist_02_r, loc=<5.5303 6.2089 1.0811>, rot=<-0.6748 -0.7324 0.0034 0.0905>}
 dyn_knot_bottom_jnt_1 {parent=dyn_knot_kinematic_jnt, loc=<2.1447 -0.8342 -0.6661>, rot=<-0.0319 -0.0958 0.1524 -0.9831>}
 dyn_knot_bottom_jnt_2 {parent=dyn_knot_bottom_jnt_1, loc=<2.3175 0.0514 0.3466>, rot=<0 0 0 1>}
 dyn_knot_top_jnt_1 {parent=dyn_knot_kinematic_jnt, loc=<1.9323 1.0565 0.3563>, rot=<-0.4455 -0.1614 -0.1793 -0.8622>}
 dyn_knot_top_jnt_2 {parent=dyn_knot_top_jnt_1, loc=<1.9518 0.144 0.2692>, rot=<-0.0038 0.0164 0.2238 -0.9745>}
 dyn_R_leafDynamic_kinematic_jnt {parent=upperarm_twist_02_r, loc=<4.2977 3.7674 -2.7449>, rot=<0.9343 -0.0039 0.3563 -0.0065>}
 dyn_R_leafDynamic_jnt_5 {parent=dyn_R_leafDynamic_kinematic_jnt, loc=<-0.2784 -0.2207 0.1553>, rot=<0.1357 0.0029 0.9302 -0.341>}
 dyn_R_leafDynamic_jnt_4 {parent=dyn_R_leafDynamic_kinematic_jnt, loc=<-0.1802 -0.6104 -0.0173>, rot=<-0.0528 -0.1893 0.9017 -0.3851>}
 dyn_R_leaf_jnt_18 {parent=upperarm_r, loc=<0.7316 4.9648 -0.8924>, rot=<0.9343 -0.0039 0.3563 -0.0065>}
 dyn_R_leaf_jnt_19 {parent=upperarm_r, loc=<0.5039 3.5297 -2.3431>, rot=<0.9343 -0.0039 0.3563 -0.0065>}
 dyn_R_leaf_jnt_20 {parent=upperarm_r, loc=<3.03 0.1677 -3.3393>, rot=<0.9343 -0.0039 0.3563 -0.0065>}
 dyn_R_leafDynamic_jnt_2 {parent=dyn_R_leaf_jnt_20, loc=<-1.7621 -0.2362 1.15>, rot=<-0.0144 0.0009 0.9733 -0.2293>}
 dyn_R_leaf_jnt_8 {parent=upperarm_r, loc=<1.0061 -1.5735 -4.1171>, rot=<0.9343 -0.0039 0.3563 -0.0065>}
 dyn_R_leafDynamic_jnt_3 {parent=dyn_R_leaf_jnt_8, loc=<-1.4357 -0.1864 0.5687>, rot=<-0.0641 0.0001 0.9979 0.0008>}
 dyn_R_leaf_jnt_5 {parent=upperarm_r, loc=<0.6524 -2.9882 -3.3719>, rot=<0.9343 -0.0039 0.3563 -0.0065>}
 dyn_R_leaf_jnt_4 {parent=upperarm_r, loc=<1.3076 -3.434 -3.0147>, rot=<0.9343 -0.0039 0.3563 -0.0065>}
 dyn_R_leaf_jnt_2 {parent=clavicle_r, loc=<-6.6668 2.5643 -4.6621>, rot=<0.9991 0.0252 0.0336 0.0008>}
 dyn_R_leaf_jnt_3 {parent=clavicle_r, loc=<-7.9977 -3.1069 -1.7022>, rot=<0.9991 0.0252 0.0336 0.0008>}
 dyn_R_leaf_jnt_1 {parent=clavicle_r, loc=<-7.5752 5.2262 -3.0532>, rot=<0.9991 0.0252 0.0336 0.0008>}
 dyn_R_leafDynamic_jnt_1 {parent=dyn_R_leaf_jnt_1, loc=<-0.3595 -0.8593 -0.3586>, rot=<-0.181 -0.2026 0.7197 -0.639>}
 neck_01 {parent=spine_05, loc=<11.3138 0.1529 0>, rot=<0 0 -0.2003 -0.9797>}
 neck_02 {parent=neck_01, loc=<4.5627 0 0>, rot=<0 0 0.0504 -0.9987>}
 head {parent=neck_02, loc=<5.3914 0 0>, rot=<0 0 0.115 -0.9934>}
 dyn_L_leaf_jnt_5 {parent=spine_05, loc=<1.4374 8.6245 -7.6303>, rot=<0.0284 -0.7065 -0.0284 -0.7065>}
 dyn_L_shoulderLeaves_jnt_1 {parent=spine_05, loc=<3.603 -10.2804 -8.9285>, rot=<0.0524 -0.9858 -0.1221 -0.1025>}
 dyn_L_shoulderLeaves_jnt_2 {parent=dyn_L_shoulderLeaves_jnt_1, loc=<3.0086 -0.4032 0.3975>, rot=<0 0 0 1>}
 dyn_R_shoulderLeaves_base_jnt {parent=spine_05, loc=<3.3329 -9.6948 4.8146>, rot=<0.0284 -0.7065 -0.0284 -0.7065>}
 dyn_R_shoulderLeaves_jnt_1 {parent=dyn_R_shoulderLeaves_base_jnt, loc=<-1.1961 -0.9872 -1.7168>, rot=<0.0795 -0.7484 0.0695 -0.6548>}
 dyn_R_shoulderLeaves_jnt_2 {parent=dyn_R_shoulderLeaves_jnt_1, loc=<2.7156 0.0711 0.3579>, rot=<0 0 -0.0506 -0.9987>}
 thigh_l {parent=pelvis, loc=<-1.0855 -0.2109 -8.7053>, rot=<0.0009 0.0355 0.0155 -0.9993>}
 calf_l {parent=thigh_l, loc=<-42.5284 -0.0019 -0.7297>, rot=<0.0006 -0.0161 -0.0391 -0.9991>}
 calf_twist_01_l {parent=calf_l, loc=<-35.8483 -1.9197 -0.5794>, rot=<0.0237 0.0095 -0.0035 -0.9997>}
 calf_twist_02_l {parent=calf_l, loc=<-19.1546 -1.7943 0>, rot=<0.0237 0.0095 -0.0035 -0.9997>}
 foot_l {parent=calf_l, loc=<-46.7402 -2.5014 -0.592>, rot=<0.0427 -0.0238 0.0353 -0.9982>}
 ball_l {parent=foot_l, loc=<-3.5658 10.8066 -0.1544>, rot=<0.0001 -0.0001 -0.7063 -0.7079>}
 foot_support_l {parent=foot_l, loc=<1.8674 -3.3442 0.1705>, rot=<0.0128 -0.7053 0.0203 -0.7085>}
 thigh_twist_01_l {parent=thigh_l, loc=<-27.952 0 0>, rot=<0.0013 0 0 -1>}
 thigh_r {parent=pelvis, loc=<-1.0853 -0.2109 8.7053>, rot=<0.0355 -0.0009 0.9993 0.0155>}
 calf_r {parent=thigh_r, loc=<42.5286 0.0019 0.7297>, rot=<0.0006 -0.0161 -0.0391 -0.9991>}
 calf_twist_01_r {parent=calf_r, loc=<35.8485 1.9197 0.5794>, rot=<0.0237 0.0095 -0.0035 -0.9997>}
 calf_twist_02_r {parent=calf_r, loc=<19.1546 1.7943 0.2638>, rot=<0.0237 0.0095 -0.0035 -0.9997>}
 dyn_bootStrap_jnt_1 {parent=calf_twist_02_r, loc=<5.5816 -3.6098 1.6383>, rot=<0.9921 0.0723 0.0176 0.1009>}
 dyn_bootStrap_jnt_2 {parent=dyn_bootStrap_jnt_1, loc=<3.6667 0.7205 0>, rot=<0 0 0 1>}
 foot_r {parent=calf_r, loc=<46.7403 2.5014 0.592>, rot=<0.0427 -0.0238 0.0353 -0.9982>}
 ball_r {parent=foot_r, loc=<3.5658 -10.8066 0.1544>, rot=<0.0001 -0.0001 -0.7063 -0.7079>}
 foot_support_r {parent=foot_r, loc=<-1.8676 3.3442 -0.1705>, rot=<0.0128 -0.7053 0.0203 -0.7085>}
 thigh_twist_01_r {parent=thigh_r, loc=<27.952 0 0>, rot=<0.0013 0 0 -1>}
 dyn_R_thighLeaf_jnt {parent=thigh_twist_01_r, loc=<5.5252 -2.3859 8.1921>, rot=<-0.3159 0.5883 0.1366 -0.7317>}
 vines_correct_R_jnt {parent=thigh_twist_01_r, loc=<-10.6203 3.6197 -7.9859>, rot=<0.7376 0.3519 -0.5752 -0.0357>}
 vines_correct_R_jnt_1 {parent=thigh_twist_01_r, loc=<-15.6759 9.5115 -2.7692>, rot=<-0.6816 0.0006 0.7317 0.0028>}
 dyn_glute_leaves_kinematic_jnt {parent=pelvis, loc=<6.2236 -12.252 -2.0767>, rot=<0.1374 -0.9905 0.0005 -0.0026>}
 dyn_glute_leaves_jnt_1 {parent=dyn_glute_leaves_kinematic_jnt, loc=<0 0 0>, rot=<0 0 0 1>}
 dyn_glute_leaves_jnt_2 {parent=dyn_glute_leaves_jnt_1, loc=<4.3429 -0.2986 -0.0216>, rot=<0 0 0 1>}
 dyn_hipleaf_jnt_1 {parent=pelvis, loc=<8.3275 -7.975 10.439>, rot=<0.5155 -0.2057 0.8246 -0.1092>}
 dyn_hipleaf_jnt_2 {parent=pelvis, loc=<9.2749 -10.903 7.0216>, rot=<0.2888 -0.6454 0.3118 -0.6347>}
 dyn_hipleaf_jnt_3 {parent=pelvis, loc=<7.82 -6.2008 13.5355>, rot=<0.5078 0.0221 0.8589 0.0625>}
 dyn_hipleaf_jnt_4 {parent=pelvis, loc=<9.6534 -7.788 13.6911>, rot=<0.5971 -0.108 0.7916 -0.0713>}
 dyn_hipleaf_jnt_5 {parent=pelvis, loc=<11.055 -11.8691 7.8339>, rot=<0.2454 -0.4696 0.5443 -0.6504>}
 dyn_hipleaf_jnt_6 {parent=pelvis, loc=<11.4717 -8.3644 10.7514>, rot=<0.1558 0.1665 0.2456 -0.9422>}
 dyn_hipleaf_jnt_7 {parent=dyn_hipleaf_jnt_6, loc=<2.8715 -1.5505 -2.7493>, rot=<0.2516 -0.4329 0.1729 -0.8482>}
 dyn_L_hipLeaf_jnt_1 {parent=pelvis, loc=<8.0315 -2.184 -12.4665>, rot=<-0.1316 -0.2772 -0.1094 -0.9454>}
 dyn_L_hipLeaf_jnt_2 {parent=pelvis, loc=<8.3543 -1.9474 -13.0465>, rot=<-0.1547 -0.6244 -0.1054 -0.7583>}
 dyn_L_hipLeaf_jnt_3 {parent=pelvis, loc=<5.1074 -8.7025 -13.7031>, rot=<0.2218 -0.7343 0.444 -0.463>}
 dyn_backButt_leaf_jnt {parent=pelvis, loc=<4.9222 -13.9757 8.6663>, rot=<0.1112 0.3955 0.0595 -0.9098>}
 ik_foot_root {parent=root, loc=<0 0 0>, rot=<0 0 0 1>}
 ik_foot_l {parent=ik_foot_root, loc=<14.8262 -7.3545 8.296>, rot=<0.0357 0.7045 0.0284 -0.7082>}
 ik_foot_r {parent=ik_foot_root, loc=<-14.8262 -7.3546 8.2958>, rot=<0.7082 0.0284 -0.7045 0.0357>}
 ik_hand_root {parent=root, loc=<0 0 0>, rot=<0 0 0 1>}
 ik_hand_gun {parent=ik_hand_root, loc=<-48.1744 3.3546 104.4381>, rot=<0.6276 0.1309 0.4183 -0.6434>}
 ik_hand_l {parent=ik_hand_gun, loc=<59.3272 67.6997 34.3524>, rot=<-0.6158 -0.7027 -0.3565 0>}
 ik_hand_r {parent=ik_hand_gun, loc=<0 0 0>, rot=<0 0 0 1>}
]
