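# Physics asset for the female medium Moss Bomber player cosmetic body, simulating its leaves around the hips and rear.
# Capsule bodies on the dynamic hip, back and glute leaf joints hang off a single pelvis capsule through swing-limited constraints.
# Only the glute leaves form a two-link chain, while every other leaf constrains directly to the pelvis.
# Twist is locked on the glute, back and left-hip leaves and narrowly limited on the numbered hip leaves.
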PhysicsAsset /BRCosmetics/Characters/Player/Female/Medium/Bodies/F_MED_Moss_Bomber/Meshes/F_MED_Moss_Bomber_Physics
bodies {
 dyn_L_hipLeaf_jnt_1 [{shapeType=Capsule, length=3, radius=1.3157, center=<1.9127 0 0>, rotation=<rot 90 0 0>}]
 dyn_L_hipLeaf_jnt_2 [{shapeType=Capsule, length=3, radius=1.7625, center=<1.9833 0 0>, rotation=<rot 90 0 0>}]
 dyn_L_hipLeaf_jnt_3 [{shapeType=Capsule, length=3, radius=1.8, center=<1.0973 0 -1.2803>, rotation=<rot 45.1866 0 0>}]
 dyn_backButt_leaf_jnt [{shapeType=Capsule, length=2.9909, radius=1.6489, center=<2.4273 0 0>, rotation=<rot 90 0 0>}]
 dyn_glute_leaves_jnt_1 [{shapeType=Capsule, length=3, radius=2, center=<1.051 0 0>, rotation=<rot 90 0 0>}]
 dyn_glute_leaves_jnt_2 [{shapeType=Capsule, length=3.8054, radius=2.6946, center=<3.1633 0 0>, rotation=<rot 90 0 0>}]
 dyn_hipleaf_jnt_1 [{shapeType=Capsule, length=4, radius=0.7, center=<0.1375 0 0>, rotation=<rot 90 0 0>}]
 dyn_hipleaf_jnt_2 [{shapeType=Capsule, length=5, radius=0.5, center=<0.1375 0 0>, rotation=<rot 90 0 0>}]
 dyn_hipleaf_jnt_3 [{shapeType=Capsule, length=4, radius=0.55, center=<0.1375 0 0>, rotation=<rot 90 0 0>}]
 dyn_hipleaf_jnt_4 [{shapeType=Capsule, length=4, radius=0.5, center=<0.1375 0 0>, rotation=<rot 90 0 0>}]
 dyn_hipleaf_jnt_5 [{shapeType=Capsule, length=4, radius=0.6, center=<0.1375 0 0>, rotation=<rot 90 0 0>}]
 dyn_hipleaf_jnt_6 [{shapeType=Capsule, length=4, radius=1, center=<0.1375 0 0>, rotation=<rot 90 0 0>}]
 dyn_hipleaf_jnt_7 [{shapeType=Capsule, length=4, radius=0.7, center=<0.1375 0 0>, rotation=<rot 90 0 0>}]
 pelvis [{shapeType=Capsule, length=8.0283, radius=9.9717, center=<0 -1.46 0>}]
}
constraints [
 {
  bone1Name=dyn_glute_leaves_jnt_2
  bone2Name=dyn_glute_leaves_jnt_1
  swing1LimitDegrees=40
  swing2LimitDegrees=20
  twistLimitDegrees=45
  twistMotion=Locked
 }
 {bone1Name=dyn_glute_leaves_jnt_1, bone2Name=pelvis, swing1LimitDegrees=40, swing2LimitDegrees=20, twistLimitDegrees=45, twistMotion=Locked}
 {bone1Name=dyn_hipleaf_jnt_1, bone2Name=pelvis, swing1LimitDegrees=25, swing2LimitDegrees=9, twistLimitDegrees=1}
 {bone1Name=dyn_hipleaf_jnt_2, bone2Name=pelvis, swing1LimitDegrees=25, swing2LimitDegrees=9, twistLimitDegrees=1}
 {bone1Name=dyn_hipleaf_jnt_3, bone2Name=pelvis, swing1LimitDegrees=25, swing2LimitDegrees=9, twistLimitDegrees=1}
 {bone1Name=dyn_hipleaf_jnt_4, bone2Name=pelvis, swing1LimitDegrees=25, swing2LimitDegrees=9, twistLimitDegrees=1}
 {bone1Name=dyn_hipleaf_jnt_5, bone2Name=pelvis, swing1LimitDegrees=25, swing2LimitDegrees=9, twistLimitDegrees=1}
 {bone1Name=dyn_hipleaf_jnt_6, bone2Name=pelvis, swing1LimitDegrees=40, swing2LimitDegrees=9, twistLimitDegrees=1}
 {bone1Name=dyn_hipleaf_jnt_7, bone2Name=pelvis, swing1LimitDegrees=25, swing2LimitDegrees=9, twistLimitDegrees=1}
 {bone1Name=dyn_L_hipLeaf_jnt_1, bone2Name=pelvis, swing1LimitDegrees=20, swing2LimitDegrees=20, twistLimitDegrees=45, twistMotion=Locked}
 {bone1Name=dyn_L_hipLeaf_jnt_2, bone2Name=pelvis, swing1LimitDegrees=15, swing2LimitDegrees=15, twistLimitDegrees=45, twistMotion=Locked}
 {bone1Name=dyn_backButt_leaf_jnt, bone2Name=pelvis, swing1LimitDegrees=15, swing2LimitDegrees=40, twistLimitDegrees=45, twistMotion=Locked}
 {bone1Name=dyn_L_hipLeaf_jnt_3, bone2Name=pelvis, swing1LimitDegrees=30, swing2LimitDegrees=30, twistLimitDegrees=45, twistMotion=Locked}
]
