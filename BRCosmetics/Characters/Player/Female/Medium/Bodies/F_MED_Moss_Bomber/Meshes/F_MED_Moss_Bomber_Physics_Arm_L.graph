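# Physics asset for the left arm of the female medium Moss Bomber body, set up for secondary motion on its leaves.
# Capsules on clavicle, upper arm and twist bone form a free-rotating chain, alongside tiny capsules on leaf joints.
# Dynamic leaf capsules constrain to leaf joints or the twist bone, except one chained to another dynamic leaf.
# Leaf constraints share tight, asymmetric swing limits with twist locked.
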PhysicsAsset /BRCosmetics/Characters/Player/Female/Medium/Bodies/F_MED_Moss_Bomber/Meshes/F_MED_Moss_Bomber_Physics_Arm_L
bodies {
 clavicle_l [{shapeType=Capsule, length=7, radius=3, center=<6.1978 0 -1.9914>, rotation=<rot 82.7458 -0.0003 -0.0003>}]
 dyn_L_leafDynamic_jnt_1 [{shapeType=Capsule, length=1.6478, radius=1.2593, center=<2.4472 0 0>, rotation=<rot 90 0 0>}]
 dyn_L_leafDynamic_jnt_10 [{shapeType=Capsule, length=1.9582, radius=1.7641, center=<1.8027 0 0>, rotation=<rot 90 0 0>}]
 dyn_L_leafDynamic_jnt_2 [{shapeType=Capsule, length=0.505, radius=0.505, center=<0.9958 0 0>, rotation=<rot 90 0 0>}]
 dyn_L_leafDynamic_jnt_3 [{shapeType=Capsule, length=1.9315, radius=1.4736, center=<1.72 0 0>, rotation=<rot 90 0 0>}]
 dyn_L_leafDynamic_jnt_4 [{shapeType=Capsule, length=1.2869, radius=1.0449, center=<1.5871 0 0>, rotation=<rot 90 0 0>}]
 dyn_L_leafDynamic_jnt_5 [{shapeType=Capsule, length=1.709, radius=1.3612, center=<1.7542 0 0>, rotation=<rot 90 0 0>}]
 dyn_L_leafDynamic_jnt_6 [{shapeType=Capsule, length=1.6783, radius=1.135, center=<2.1658 0 0>, rotation=<rot 90 0 0>}]
 dyn_L_leafDynamic_jnt_7 [{shapeType=Capsule, length=1.4363, radius=1.0951, center=<1.3765 0 0>, rotation=<rot 90 0 0>}]
 dyn_L_leafDynamic_jnt_8 [{shapeType=Capsule, length=1.8038, radius=1.345, center=<1.2463 0 0>, rotation=<rot 90 0 0>}]
 dyn_L_leafDynamic_jnt_9 [{shapeType=Capsule, length=1.6893, radius=1.4731, center=<2.0689 0 0>, rotation=<rot 90 0 0>}]
 dyn_L_leaf_jnt_1 [{shapeType=Capsule, length=0.5, radius=0.5}]
 dyn_L_leaf_jnt_10 [{shapeType=Capsule, length=0.5, radius=0.5}]
 dyn_L_leaf_jnt_11 [{shapeType=Capsule, length=0.5, radius=0.5}]
 dyn_L_leaf_jnt_12 [{shapeType=Capsule, length=0.5, radius=0.5}]
 dyn_L_leaf_jnt_13 [{shapeType=Capsule, length=0.5, radius=0.5}]
 dyn_L_leaf_jnt_14 [{shapeType=Capsule, length=0.5, radius=0.5}]
 dyn_L_leaf_jnt_15 [{shapeType=Capsule, length=0.5, radius=0.5}]
 dyn_L_leaf_jnt_16 [{shapeType=Capsule, length=0.5, radius=0.5}]
 dyn_L_leaf_jnt_2 [{shapeType=Capsule, length=0.5, radius=0.5}]
 dyn_L_leaf_jnt_3 [{shapeType=Capsule, length=0.5, radius=0.5}]
 dyn_L_leaf_jnt_4 [{shapeType=Capsule, length=0.5, radius=0.5}]
 dyn_L_leaf_jnt_6 [{shapeType=Capsule, length=0.5, radius=0.5}]
 dyn_L_leaf_jnt_7 [{shapeType=Capsule, length=0.5, radius=0.5}]
 dyn_L_leaf_jnt_8 [{shapeType=Capsule, length=0.5, radius=0.5}]
 dyn_L_leaf_jnt_9 [{shapeType=Capsule, length=0.5, radius=0.5}]
 upperarm_l [{shapeType=Capsule, length=10, radius=4, center=<10.2121 -0.0001 -0.6728>, rotation=<rot 90 0 0>}]
 upperarm_twist_02_l [{shapeType=Capsule, length=5, radius=2, rotation=<rot 90 0 0>}]
}
constraints [
 {
  bone1Name=dyn_L_leafDynamic_jnt_7
  bone2Name=upperarm_twist_02_l
  swing1LimitDegrees=15
  swing2LimitDegrees=10
  twistLimitDegrees=45
  twistMotion=Locked
 }
 {
  bone1Name=dyn_L_leafDynamic_jnt_8
  bone2Name=upperarm_twist_02_l
  swing1LimitDegrees=15
  swing2LimitDegrees=10
  twistLimitDegrees=45
  twistMotion=Locked
 }
 {
  bone1Name=dyn_L_leafDynamic_jnt_1
  bone2Name=dyn_L_leaf_jnt_1
  swing1LimitDegrees=15
  swing2LimitDegrees=10
  twistLimitDegrees=45
  twistMotion=Locked
 }
 {
  bone1Name=dyn_L_leafDynamic_jnt_4
  bone2Name=dyn_L_leaf_jnt_2
  swing1LimitDegrees=15
  swing2LimitDegrees=10
  twistLimitDegrees=45
  twistMotion=Locked
 }
 {
  bone1Name=dyn_L_leafDynamic_jnt_6
  bone2Name=dyn_L_leaf_jnt_4
  swing1LimitDegrees=15
  swing2LimitDegrees=10
  twistLimitDegrees=45
  twistMotion=Locked
 }
 {
  bone1Name=dyn_L_leafDynamic_jnt_5
  bone2Name=dyn_L_leaf_jnt_10
  swing1LimitDegrees=15
  swing2LimitDegrees=10
  twistLimitDegrees=45
  twistMotion=Locked
 }
 {
  bone1Name=dyn_L_leafDynamic_jnt_9
  bone2Name=dyn_L_leaf_jnt_11
  swing1LimitDegrees=15
  swing2LimitDegrees=10
  twistLimitDegrees=45
  twistMotion=Locked
 }
 {
  bone1Name=dyn_L_leafDynamic_jnt_10
  bone2Name=dyn_L_leaf_jnt_12
  swing1LimitDegrees=15
  swing2LimitDegrees=10
  twistLimitDegrees=45
  twistMotion=Locked
 }
 {
  bone1Name=dyn_L_leafDynamic_jnt_2
  bone2Name=dyn_L_leaf_jnt_16
  swing1LimitDegrees=15
  swing2LimitDegrees=10
  twistLimitDegrees=45
  twistMotion=Locked
 }
 {
  bone1Name=dyn_L_leafDynamic_jnt_3
  bone2Name=dyn_L_leafDynamic_jnt_2
  swing1LimitDegrees=15
  swing2LimitDegrees=10
  twistLimitDegrees=45
  twistMotion=Locked
 }
 {
  bone1Name=upperarm_l
  bone2Name=clavicle_l
  swing1LimitDegrees=45
  swing1Motion=Free
  swing2LimitDegrees=45
  swing2Motion=Free
  twistLimitDegrees=45
  twistMotion=Free
 }
 {
  bone1Name=upperarm_twist_02_l
  bone2Name=upperarm_l
  swing1LimitDegrees=45
  swing1Motion=Free
  swing2LimitDegrees=45
  swing2Motion=Free
  twistLimitDegrees=45
  twistMotion=Free
 }
]
