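# Physics asset for the leaves on the right arm of the Moss Bomber, a medium female player cosmetic body.
# Three capsule bodies on leafDynamic joints each hang off a separate leaf joint through constraints with narrow swing limits and locked twist.
# Twenty small placeholder-sized capsules cover the leaf joints, and one large capsule covers the upper arm.
# Only one leaf joint is constrained to the upper arm, with free swing and twist.
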PhysicsAsset /BRCosmetics/Characters/Player/Female/Medium/Bodies/F_MED_Moss_Bomber/Meshes/F_MED_Moss_Bomber_Physics_Arm_R
bodies {
 dyn_R_leafDynamic_jnt_1 [{shapeType=Capsule, length=1.9884, radius=1.5178, center=<1.7388 0 0>, rotation=<rot 90 0 0>}]
 dyn_R_leafDynamic_jnt_2 [{shapeType=Capsule, length=1.8958, radius=1.4789, center=<2.0124 0 0>, rotation=<rot 90 0 0>}]
 dyn_R_leafDynamic_jnt_3 [{shapeType=Capsule, length=1.7065, radius=1.3573, center=<1.7273 0 0>, rotation=<rot 90 0 0>}]
 dyn_R_leaf_jnt_1 [{shapeType=Capsule, length=0.5, radius=0.5}]
 dyn_R_leaf_jnt_10 [{shapeType=Capsule, length=0.5, radius=0.5}]
 dyn_R_leaf_jnt_11 [{shapeType=Capsule, length=0.5, radius=0.5}]
 dyn_R_leaf_jnt_12 [{shapeType=Capsule, length=0.5, radius=0.5}]
 dyn_R_leaf_jnt_13 [{shapeType=Capsule, length=0.5, radius=0.5}]
 dyn_R_leaf_jnt_14 [{shapeType=Capsule, length=0.5, radius=0.5}]
 dyn_R_leaf_jnt_15 [{shapeType=Capsule, length=0.5, radius=0.5}]
 dyn_R_leaf_jnt_16 [{shapeType=Capsule, length=0.5, radius=0.5}]
 dyn_R_leaf_jnt_17 [{shapeType=Capsule, length=0.5, radius=0.5}]
 dyn_R_leaf_jnt_18 [{shapeType=Capsule, length=0.5, radius=0.5}]
 dyn_R_leaf_jnt_19 [{shapeType=Capsule, length=0.5, radius=0.5}]
 dyn_R_leaf_jnt_2 [{shapeType=Capsule, length=0.5, radius=0.5}]
 dyn_R_leaf_jnt_20 [{shapeType=Capsule, length=0.5, radius=0.5}]
 dyn_R_leaf_jnt_3 [{shapeType=Capsule, length=0.5, radius=0.5}]
 dyn_R_leaf_jnt_4 [{shapeType=Capsule, length=0.5, radius=0.5}]
 dyn_R_leaf_jnt_5 [{shapeType=Capsule, length=0.5, radius=0.5}]
 dyn_R_leaf_jnt_6 [{shapeType=Capsule, length=0.5, radius=0.5}]
 dyn_R_leaf_jnt_7 [{shapeType=Capsule, length=0.5, radius=0.5}]
 dyn_R_leaf_jnt_8 [{shapeType=Capsule, length=0.5, radius=0.5}]
 dyn_R_leaf_jnt_9 [{shapeType=Capsule, length=0.5, radius=0.5}]
 upperarm_r [{shapeType=Capsule, length=26.8448, radius=5, center=<-14.0795 0 0>, rotation=<rot -86.2148 -0.0001 0.0001>}]
}
constraints [
 {
  bone1Name=dyn_R_leafDynamic_jnt_1
  bone2Name=dyn_R_leaf_jnt_1
  swing1LimitDegrees=15
  swing2LimitDegrees=10
  twistLimitDegrees=45
  twistMotion=Locked
 }
 {
  bone1Name=dyn_R_leafDynamic_jnt_3
  bone2Name=dyn_R_leaf_jnt_8
  swing1LimitDegrees=15
  swing2LimitDegrees=10
  twistLimitDegrees=45
  twistMotion=Locked
 }
 {
  bone1Name=dyn_R_leafDynamic_jnt_2
  bone2Name=dyn_R_leaf_jnt_20
  swing1LimitDegrees=15
  swing2LimitDegrees=10
  twistLimitDegrees=45
  twistMotion=Locked
 }
 {
  bone1Name=dyn_R_leaf_jnt_12
  bone2Name=upperarm_r
  swing1LimitDegrees=45
  swing1Motion=Free
  swing2LimitDegrees=45
  swing2Motion=Free
  twistLimitDegrees=45
  twistMotion=Free
 }
]
