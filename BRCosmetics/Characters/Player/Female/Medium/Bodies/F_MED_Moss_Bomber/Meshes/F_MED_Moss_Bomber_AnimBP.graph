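# Animation Blueprint for the F_MED_Moss_Bomber cosmetic body, driving its pose and secondary motion of leaves, knots and boot straps.
# Copies the attached parent mesh's pose through a Control Rig, then chains RigidBody simulations based at the pelvis and each upper arm.
# Anim Dynamics nodes add damped, angle-limited sway to dangling bones.
# Multi-target Constraint nodes place leaf joints by weighted blends of nearby body bones, resembling skin weights.
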AnimBlueprint /BRCosmetics/Characters/Player/Female/Medium/Bodies/F_MED_Moss_Bomber/Meshes/F_MED_Moss_Bomber_AnimBP
target_skeleton=&/F_MED_Moss_Bomber_Skeleton
graphs [
 AnimGraph {nodes [
   n1 = Control Rig(Source=Copy Pose From Mesh() {bUseAttachedParent=true}) {
    alphaBoolBlend {blendOption=Linear}
    controlRigAssetReference {blueprintRigClass=&/F_MED_Moss_Bomber_CtrlRig.F_MED_Moss_Bomber_CtrlRig_C}
   }
   n2 = RigidBody!(ComponentPose=Local To Component(LocalPose=@n1), Alpha, OverrideWorldGravity:Vector="0,0,-980", ExternalForce) -> (Pose) {
    bClampLinearTranslationLimitToRefPose=true
    bOverrideWorldGravity=true
    overridePhysicsAsset=&/F_MED_Moss_Bomber_Physics
    simulationSpace=BaseBoneSpace
    alphaBoolBlend {blendOption=Linear}
    baseBoneRef {boneName=pelvis}
    overrideWorldGravity {z=-980}
    simSpaceSettings {
     maxAngularAcceleration=500
     maxAngularVelocity=3
     maxLinearAcceleration=400
     maxLinearVelocity=250
     velocityScaleZ=0
     worldAlpha=1
     externalLinearDragV=<0.1 0.1>
    }
   }
   n3 = RigidBody(ComponentPose=@n2, OverrideWorldGravity="0,0,-980") {
    bClampLinearTranslationLimitToRefPose=true
    bOverrideWorldGravity=true
    overridePhysicsAsset=&/F_MED_Moss_Bomber_Physics_Arm_L
    simulationSpace=BaseBoneSpace
    alphaBoolBlend {blendOption=Linear}
    baseBoneRef {boneName=upperarm_l}
    overrideWorldGravity {z=-980}
    simSpaceSettings {
     maxAngularAcceleration=500
     maxAngularVelocity=3
     maxLinearAcceleration=400
     maxLinearVelocity=250
     velocityScaleZ=0
     worldAlpha=1
     externalLinearDragV=<0.1 0.1>
    }
   }
   n4 = RigidBody(ComponentPose=@n3, OverrideWorldGravity="0,0,-980") {
    bClampLinearTranslationLimitToRefPose=true
    bOverrideWorldGravity=true
    overridePhysicsAsset=&/F_MED_Moss_Bomber_Physics_Arm_R
    simulationSpace=BaseBoneSpace
    alphaBoolBlend {blendOption=Linear}
    baseBoneRef {boneName=upperarm_r}
    overrideWorldGravity {z=-980}
    simSpaceSettings {
     maxAngularAcceleration=500
     maxAngularVelocity=3
     maxLinearAcceleration=400
     maxLinearVelocity=250
     velocityScaleZ=0
     worldAlpha=1
     externalLinearDragV=<0.1 0.1>
    }
   }
   n5 = "Constraint - dyn_L_shoulderLeaves_jnt_1 (PPPPPPPPPPPPPP)"!(
    ComponentPose=@n4
    Alpha=1
    ConstraintWeights_0=0
    ConstraintWeights_1:float=0.4052
    ConstraintWeights_2:float=0
    ConstraintWeights_3:float=0
    ConstraintWeights_4:float=0.0001
    ConstraintWeights_5:float=0.0027
    ConstraintWeights_6:float=0.1778
    ConstraintWeights_7:float=0.3389
    ConstraintWeights_8:float=0.0652
    ConstraintWeights_9:float=0
    ConstraintWeights_10:float=0.0096
    ConstraintWeights_11:float=0
    ConstraintWeights_12:float=0
    ConstraintWeights_13:float=0.0003
   ) -> (Pose) {
    alphaBoolBlend {blendOption=Linear}
    boneToModify {boneName=dyn_L_shoulderLeaves_jnt_1}
    constraintSetup [
     {targetBone {boneName=upperarm_r}}
     {targetBone {boneName=clavicle_l}}
     {targetBone {boneName=thigh_r}}
     {targetBone {boneName=thigh_twist_01_r}}
     {targetBone {boneName=spine_01}}
     {targetBone {boneName=spine_02}}
     {targetBone {boneName=spine_03}}
     {targetBone {boneName=spine_04}}
     {targetBone {boneName=spine_05}}
     {targetBone {boneName=thigh_l}}
     {targetBone {boneName=clavicle_r}}
     {targetBone {boneName=pelvis}}
     {targetBone {boneName=thigh_twist_01_l}}
     {targetBone {boneName=upperarm_l}}
    ]
    constraintWeights [1,1,1,1,1,1,1,1,1,1,1,1,1,1]
   }
   n6 = Anim Dynamics - Bone: dyn_L_shoulderLeaves_jnt_1!(ComponentPose=@n5, Alpha=1, GravityOverride:Vector="0,0,0", ExternalForce="0,0,0") -> (Pose) {
    angularDampingOverride=0.9999
    bGravityOverrideInSimSpace=true
    bOverrideAngularDamping=true
    bUseGravityOverride=true
    gravityScale=0.4
    alphaBoolBlend {blendOption=Linear}
    boundBone {boneName=dyn_L_shoulderLeaves_jnt_1}
    componentAppliedLinearAccClamp=<2000 2000 2000>
    componentLinearAccScale=<0.4 0.4 0.48>
    componentLinearVelScale=<0.2 0.2 0.2>
    physicsBodyDefinitions [{
      sphereCollisionRadius=0
      boundBone {boneName=dyn_L_shoulderLeaves_jnt_1}
      boxExtents=<3 3 3>
      constraintSetup {angularLimitsMax=<0 0 0>, angularLimitsMin=<0 -10 -20>, angularTarget=<1 0 0>}
      localJointOffset=<0.1 0 0>
     }]
    relativeSpaceBone {boneName=spine_05}
   }
   n7 = Anim Dynamics - Bone: dyn_L_shoulderLeaves_jnt_2!(ComponentPose=@n6, Alpha=1, GravityOverride:Vector="0,0,0", ExternalForce="0,0,0") -> (Pose) {
    angularDampingOverride=0.9999
    bGravityOverrideInSimSpace=true
    bOverrideAngularDamping=true
    bUseGravityOverride=true
    gravityScale=0.4
    alphaBoolBlend {blendOption=Linear}
    boundBone {boneName=dyn_L_shoulderLeaves_jnt_2}
    componentAppliedLinearAccClamp=<2000 2000 2000>
    componentLinearAccScale=<0.4 0.4 0.6>
    componentLinearVelScale=<0.2 0.2 0.2>
    physicsBodyDefinitions [{
      sphereCollisionRadius=0
      boundBone {boneName=dyn_L_shoulderLeaves_jnt_2}
      boxExtents=<3 3 3>
      constraintSetup {angularLimitsMax=<0 0 0>, angularLimitsMin=<0 -10 -20>, angularTarget=<1 0 0>}
      localJointOffset=<0.1 0 0>
     }]
    relativeSpaceBone {boneName=spine_05}
   }
   n8 = Anim Dynamics - Bone: dyn_R_thighLeaf_jnt!(ComponentPose=@n7, Alpha=1, GravityOverride:Vector="0,0,0", ExternalForce="0,0,0") -> (Pose) {
    angularDampingOverride=0.9999
    bGravityOverrideInSimSpace=true
    bOverrideAngularDamping=true
    bUseGravityOverride=true
    gravityScale=0.4
    alphaBoolBlend {blendOption=Linear}
    boundBone {boneName=dyn_R_thighLeaf_jnt}
    componentAppliedLinearAccClamp=<2000 2000 2000>
    componentLinearAccScale=<0.4 0.4 0.48>
    componentLinearVelScale=<0.2 0.2 0.2>
    physicsBodyDefinitions [{
      sphereCollisionRadius=0
      boundBone {boneName=dyn_R_thighLeaf_jnt}
      boxExtents=<3 3 3>
      constraintSetup {angularLimitsMax=<10 20 0>, angularLimitsMin=<-10 -10 -20>, angularTarget=<1 0 0>}
      localJointOffset=<0.2 0 0>
     }]
    relativeSpaceBone {boneName=spine_05}
   }
   n9 = Anim Dynamics - Bone: dyn_R_leafDynamic_jnt_4!(ComponentPose=@n8, Alpha=1, GravityOverride:Vector="0,0,0", ExternalForce="0,0,0") -> (Pose) {
    angularDampingOverride=1
    bGravityOverrideInSimSpace=true
    bOverrideAngularDamping=true
    bUseGravityOverride=true
    gravityScale=0.4
    alphaBoolBlend {blendOption=Linear}
    boundBone {boneName=dyn_R_leafDynamic_jnt_4}
    componentAppliedLinearAccClamp=<2000 2000 2000>
    componentLinearAccScale=<0.4 0.4 0.48>
    componentLinearVelScale=<0.2 0.2 0.2>
    physicsBodyDefinitions [{
      sphereCollisionRadius=0
      boundBone {boneName=dyn_R_leafDynamic_jnt_4}
      boxExtents=<3 3 3>
      constraintSetup {angularLimitsMax=<0 20 0>, angularLimitsMin=<0 10 -20>, angularTarget=<1 0 0>}
      localJointOffset=<0.05 0 0>
     }]
    relativeSpaceBone {boneName=spine_05}
   }
   n10 = Anim Dynamics - Bone: dyn_R_leafDynamic_jnt_5!(ComponentPose=@n9, Alpha=1, GravityOverride:Vector="0,0,0", ExternalForce="0,0,0") -> (Pose) {
    angularDampingOverride=0.9999
    bGravityOverrideInSimSpace=true
    bOverrideAngularDamping=true
    bUseGravityOverride=true
    gravityScale=0.4
    alphaBoolBlend {blendOption=Linear}
    boundBone {boneName=dyn_R_leafDynamic_jnt_5}
    componentAppliedLinearAccClamp=<2000 2000 2000>
    componentLinearAccScale=<0.4 0.4 0.48>
    componentLinearVelScale=<0.2 0.2 0.2>
    physicsBodyDefinitions [{
      sphereCollisionRadius=0
      boundBone {boneName=dyn_R_leafDynamic_jnt_5}
      boxExtents=<3 3 3>
      constraintSetup {angularLimitsMax=<0 20 0>, angularLimitsMin=<0 -10 -20>, angularTarget=<1 0 0>}
      localJointOffset=<0.1 0 0>
     }]
    relativeSpaceBone {boneName=spine_05}
   }
   n11 = Anim Dynamics - Bone: dyn_knot_top_jnt_1!(ComponentPose=@n10, Alpha=1, GravityOverride:Vector="0,0,0", ExternalForce="0,0,0") -> (Pose) {
    angularDampingOverride=0.9999
    bGravityOverrideInSimSpace=true
    bOverrideAngularDamping=true
    bUseGravityOverride=true
    gravityScale=0.4
    alphaBoolBlend {blendOption=Linear}
    boundBone {boneName=dyn_knot_top_jnt_1}
    componentAppliedLinearAccClamp=<2000 2000 2000>
    componentLinearAccScale=<0.4 0.4 0.48>
    componentLinearVelScale=<0.2 0.2 0.2>
    physicsBodyDefinitions [{
      sphereCollisionRadius=0
      boundBone {boneName=dyn_knot_top_jnt_1}
      boxExtents=<3 3 3>
      constraintSetup {angularLimitsMax=<10 20 0>, angularLimitsMin=<-10 -10 -20>, angularTarget=<1 0 0>}
      localJointOffset=<0.1 0 0>
     }]
    relativeSpaceBone {boneName=spine_05}
   }
   n12 = Anim Dynamics - Bone: dyn_knot_bottom_jnt_1!(ComponentPose=@n11, Alpha=1, GravityOverride:Vector="0,0,0", ExternalForce="0,0,0") -> (Pose) {
    angularDampingOverride=0.9999
    bGravityOverrideInSimSpace=true
    bOverrideAngularDamping=true
    bUseGravityOverride=true
    gravityScale=0.4
    alphaBoolBlend {blendOption=Linear}
    boundBone {boneName=dyn_knot_bottom_jnt_1}
    componentAppliedLinearAccClamp=<2000 2000 2000>
    componentLinearAccScale=<0.4 0.4 0.48>
    componentLinearVelScale=<0.2 0.2 0.2>
    physicsBodyDefinitions [{
      sphereCollisionRadius=0
      boundBone {boneName=dyn_knot_bottom_jnt_1}
      boxExtents=<3 3 3>
      constraintSetup {angularLimitsMax=<10 20 0>, angularLimitsMin=<-10 -10 -20>, angularTarget=<1 0 0>}
      localJointOffset=<0.1 0 0>
     }]
    relativeSpaceBone {boneName=spine_05}
   }
   n13 = Anim Dynamics - Bone: dyn_knot_top_jnt_2!(ComponentPose=@n12, Alpha=1, GravityOverride:Vector="0,0,0", ExternalForce="0,0,0") -> (Pose) {
    angularDampingOverride=0.9999
    bGravityOverrideInSimSpace=true
    bOverrideAngularDamping=true
    bUseGravityOverride=true
    gravityScale=0.4
    alphaBoolBlend {blendOption=Linear}
    boundBone {boneName=dyn_knot_top_jnt_2}
    componentAppliedLinearAccClamp=<2000 2000 2000>
    componentLinearAccScale=<0.4 0.4 0.48>
    componentLinearVelScale=<0.2 0.2 0.2>
    physicsBodyDefinitions [{
      sphereCollisionRadius=0
      boundBone {boneName=dyn_knot_top_jnt_2}
      boxExtents=<3 3 3>
      constraintSetup {angularLimitsMax=<10 20 0>, angularLimitsMin=<-10 -10 -20>, angularTarget=<1 0 0>}
      localJointOffset=<0.1 0 0>
     }]
    relativeSpaceBone {boneName=spine_05}
   }
   n14 = Anim Dynamics - Bone: dyn_knot_bottom_jnt_2!(ComponentPose=@n13, Alpha=1, GravityOverride:Vector="0,0,0", ExternalForce="0,0,0") -> (Pose) {
    angularDampingOverride=0.9999
    bGravityOverrideInSimSpace=true
    bOverrideAngularDamping=true
    bUseGravityOverride=true
    gravityScale=0.4
    alphaBoolBlend {blendOption=Linear}
    boundBone {boneName=dyn_knot_bottom_jnt_2}
    componentAppliedLinearAccClamp=<2000 2000 2000>
    componentLinearAccScale=<0.4 0.4 0.48>
    componentLinearVelScale=<0.2 0.2 0.2>
    physicsBodyDefinitions [{
      sphereCollisionRadius=0
      boundBone {boneName=dyn_knot_bottom_jnt_2}
      boxExtents=<3 3 3>
      constraintSetup {angularLimitsMax=<10 20 0>, angularLimitsMin=<-10 -10 -20>, angularTarget=<1 0 0>}
      localJointOffset=<0.1 0 0>
     }]
    relativeSpaceBone {boneName=spine_05}
   }
   n15 = Anim Dynamics - Bone: dyn_R_shoulderLeaves_jnt_1!(ComponentPose=@n14, Alpha=1, GravityOverride:Vector="0,0,0", ExternalForce="0,0,0") -> (Pose) {
    angularDampingOverride=0.9999
    bGravityOverrideInSimSpace=true
    bOverrideAngularDamping=true
    bUseGravityOverride=true
    bUsePlanarLimit=false
    gravityScale=0.4
    alphaBoolBlend {blendOption=Linear}
    boundBone {boneName=dyn_R_shoulderLeaves_jnt_1}
    componentAppliedLinearAccClamp=<2000 2000 2000>
    componentLinearAccScale=<0.4 0.4 0.48>
    componentLinearVelScale=<0.2 0.2 0.2>
    physicsBodyDefinitions [{
      sphereCollisionRadius=0
      boundBone {boneName=dyn_R_shoulderLeaves_jnt_1}
      boxExtents=<3 3 3>
      constraintSetup {angularLimitsMax=<0 10 0>, angularLimitsMin=<0 -10 -20>, angularTarget=<1 0 0>}
      localJointOffset=<0.1 0 0>
     }]
    relativeSpaceBone {boneName=spine_05}
   }
   n16 = Anim Dynamics - Bone: dyn_R_shoulderLeaves_jnt_2!(ComponentPose=@n15, Alpha=1, GravityOverride:Vector="0,0,0", ExternalForce="0,0,0") -> (Pose) {
    angularDampingOverride=0.9999
    bGravityOverrideInSimSpace=true
    bOverrideAngularDamping=true
    bUseGravityOverride=true
    bUsePlanarLimit=false
    gravityScale=0.4
    alphaBoolBlend {blendOption=Linear}
    boundBone {boneName=dyn_R_shoulderLeaves_jnt_2}
    componentAppliedLinearAccClamp=<2000 2000 2000>
    componentLinearAccScale=<0.4 0.4 0.48>
    componentLinearVelScale=<0.2 0.2 0.2>
    physicsBodyDefinitions [{
      sphereCollisionRadius=0
      boundBone {boneName=dyn_R_shoulderLeaves_jnt_2}
      boxExtents=<3 3 3>
      constraintSetup {angularLimitsMax=<0 10 0>, angularLimitsMin=<0 -10 -20>, angularTarget=<1 0 0>}
      localJointOffset=<0.1 0 0>
     }]
    relativeSpaceBone {boneName=spine_05}
   }
   n17 = Anim Dynamics - Bone: dyn_bootStrap_jnt_1!(ComponentPose=@n16, Alpha=1, GravityOverride:Vector="0,0,0", ExternalForce="0,0,0") -> (Pose) {
    angularDampingOverride=0.9999
    bGravityOverrideInSimSpace=true
    bOverrideAngularDamping=true
    bUseGravityOverride=true
    gravityScale=0.4
    alphaBoolBlend {blendOption=Linear}
    boundBone {boneName=dyn_bootStrap_jnt_1}
    componentAppliedLinearAccClamp=<2000 2000 2000>
    componentLinearAccScale=<0.4 0.4 0.48>
    componentLinearVelScale=<0.2 0.2 0.2>
    physicsBodyDefinitions [{
      sphereCollisionRadius=0
      boundBone {boneName=dyn_bootStrap_jnt_1}
      boxExtents=<3 3 3>
      constraintSetup {angularLimitsMax=<10 20 30>, angularLimitsMin=<-10 -10 -10>, angularTarget=<1 0 0>}
      localJointOffset=<0.1 0 0>
     }]
    relativeSpaceBone {boneName=spine_05}
   }
   n18 = Anim Dynamics - Bone: dyn_bootStrap_jnt_2!(ComponentPose=@n17, Alpha=1, GravityOverride:Vector="0,0,0", ExternalForce="0,0,0") -> (Pose) {
    angularDampingOverride=0.9999
    bGravityOverrideInSimSpace=true
    bOverrideAngularDamping=true
    bUseGravityOverride=true
    gravityScale=0.4
    alphaBoolBlend {blendOption=Linear}
    boundBone {boneName=dyn_bootStrap_jnt_2}
    componentAppliedLinearAccClamp=<2000 2000 2000>
    componentLinearAccScale=<0.4 0.4 0.48>
    componentLinearVelScale=<0.2 0.2 0.2>
    physicsBodyDefinitions [{
      sphereCollisionRadius=0
      boundBone {boneName=dyn_bootStrap_jnt_2}
      boxExtents=<3 3 3>
      constraintSetup {angularLimitsMax=<10 20 30>, angularLimitsMin=<-10 -10 -10>, angularTarget=<1 0 0>}
      localJointOffset=<0.1 0 0>
     }]
    relativeSpaceBone {boneName=spine_05}
   }
   n19 = "Constraint - dyn_L_leaf_jnt_1 (PPPPPPPPPPPP)"!(
    ComponentPose=@n18
    Alpha=1
    ConstraintWeights_0=0.594
    ConstraintWeights_1:float=0.0003
    ConstraintWeights_2:float=0.0013
    ConstraintWeights_3:float=0.0459
    ConstraintWeights_4:float=0.0298
    ConstraintWeights_5:float=0.2068
    ConstraintWeights_6:float=0.0978
    ConstraintWeights_7:float=0.0008
    ConstraintWeights_8:float=0
    ConstraintWeights_9:float=0
    ConstraintWeights_10:float=0.0234
    ConstraintWeights_11:float=0
   ) -> (Pose) {
    alphaBoolBlend {blendOption=Linear}
    boneToModify {boneName=dyn_L_leaf_jnt_1}
    constraintSetup [
     {targetBone {boneName=clavicle_l}}
     {targetBone {boneName=upperarm_twist_02_l}}
     {targetBone {boneName=spine_02}}
     {targetBone {boneName=spine_03}}
     {targetBone {boneName=spine_04}}
     {targetBone {boneName=spine_05}}
     {targetBone {boneName=upperarm_twist_01_l}}
     {targetBone {boneName=clavicle_r}}
     {targetBone {boneName=lowerarm_l}}
     {targetBone {boneName=lowerarm_twist_01_l}}
     {targetBone {boneName=upperarm_l}}
     {targetBone {boneName=lowerarm_twist_02_l}}
    ]
    constraintWeights [1,1,1,1,1,1,1,1,1,1,1,1]
   }
   n20 = "Constraint - dyn_L_leaf_jnt_2 (PPPPPPPPPPPP)"!(
    ComponentPose=@n19
    Alpha=1
    ConstraintWeights_0=0.3904
    ConstraintWeights_1:float=0.0019
    ConstraintWeights_2:float=0.0023
    ConstraintWeights_3:float=0.0191
    ConstraintWeights_4:float=0.0176
    ConstraintWeights_5:float=0.0385
    ConstraintWeights_6:float=0.5147
    ConstraintWeights_7:float=0.0018
    ConstraintWeights_8:float=0
    ConstraintWeights_9:float=0
    ConstraintWeights_10:float=0.0138
    ConstraintWeights_11:float=0
   ) -> (Pose) {
    alphaBoolBlend {blendOption=Linear}
    boneToModify {boneName=dyn_L_leaf_jnt_2}
    constraintSetup [
     {targetBone {boneName=clavicle_l}}
     {targetBone {boneName=upperarm_twist_02_l}}
     {targetBone {boneName=spine_02}}
     {targetBone {boneName=spine_03}}
     {targetBone {boneName=spine_04}}
     {targetBone {boneName=spine_05}}
     {targetBone {boneName=upperarm_twist_01_l}}
     {targetBone {boneName=clavicle_r}}
     {targetBone {boneName=lowerarm_l}}
     {targetBone {boneName=lowerarm_twist_01_l}}
     {targetBone {boneName=upperarm_l}}
     {targetBone {boneName=lowerarm_twist_02_l}}
    ]
    constraintWeights [1,1,1,1,1,1,1,1,1,1,1,1]
   }
   n21 = "Constraint - dyn_L_leaf_jnt_3 (PPPPPPPPPPPP)"!(
    ComponentPose=@n20
    Alpha=1
    ConstraintWeights_0=0.339
    ConstraintWeights_1:float=0.002
    ConstraintWeights_2:float=0.0023
    ConstraintWeights_3:float=0.0168
    ConstraintWeights_4:float=0.0157
    ConstraintWeights_5:float=0.0286
    ConstraintWeights_6:float=0.5816
    ConstraintWeights_7:float=0.0017
    ConstraintWeights_8:float=0
    ConstraintWeights_9:float=0
    ConstraintWeights_10:float=0.0123
    ConstraintWeights_11:float=0
   ) -> (Pose) {
    alphaBoolBlend {blendOption=Linear}
    boneToModify {boneName=dyn_L_leaf_jnt_3}
    constraintSetup [
     {targetBone {boneName=clavicle_l}}
     {targetBone {boneName=upperarm_twist_02_l}}
     {targetBone {boneName=spine_02}}
     {targetBone {boneName=spine_03}}
     {targetBone {boneName=spine_04}}
     {targetBone {boneName=spine_05}}
     {targetBone {boneName=upperarm_twist_01_l}}
     {targetBone {boneName=clavicle_r}}
     {targetBone {boneName=lowerarm_l}}
     {targetBone {boneName=lowerarm_twist_01_l}}
     {targetBone {boneName=upperarm_l}}
     {targetBone {boneName=lowerarm_twist_02_l}}
    ]
    constraintWeights [1,1,1,1,1,1,1,1,1,1,1,1]
   }
   n22 = "Constraint - dyn_L_leaf_jnt_4 (PPPPPPPPPPPP)"!(
    ComponentPose=@n21
    Alpha=1
    ConstraintWeights_0=0.698
    ConstraintWeights_1:float=0.0002
    ConstraintWeights_2:float=0.0061
    ConstraintWeights_3:float=0.0306
    ConstraintWeights_4:float=0.0239
    ConstraintWeights_5:float=0.0687
    ConstraintWeights_6:float=0.1603
    ConstraintWeights_7:float=0.0008
    ConstraintWeights_8:float=0
    ConstraintWeights_9:float=0
    ConstraintWeights_10:float=0.0113
    ConstraintWeights_11:float=0
   ) -> (Pose) {
    alphaBoolBlend {blendOption=Linear}
    boneToModify {boneName=dyn_L_leaf_jnt_4}
    constraintSetup [
     {targetBone {boneName=clavicle_l}}
     {targetBone {boneName=upperarm_twist_02_l}}
     {targetBone {boneName=spine_02}}
     {targetBone {boneName=spine_03}}
     {targetBone {boneName=spine_04}}
     {targetBone {boneName=spine_05}}
     {targetBone {boneName=upperarm_twist_01_l}}
     {targetBone {boneName=clavicle_r}}
     {targetBone {boneName=lowerarm_l}}
     {targetBone {boneName=lowerarm_twist_01_l}}
     {targetBone {boneName=upperarm_l}}
     {targetBone {boneName=lowerarm_twist_02_l}}
    ]
    constraintWeights [1,1,1,1,1,1,1,1,1,1,1,1]
   }
   n23 = "Constraint - dyn_L_leaf_jnt_6 (PPPPPPPPPPPP)"!(
    ComponentPose=@n22
    Alpha=1
    ConstraintWeights_0=0.1821
    ConstraintWeights_1:float=0.0024
    ConstraintWeights_2:float=0.005
    ConstraintWeights_3:float=0.0222
    ConstraintWeights_4:float=0.0209
    ConstraintWeights_5:float=0.0176
    ConstraintWeights_6:float=0.7336
    ConstraintWeights_7:float=0.0033
    ConstraintWeights_8:float=0
    ConstraintWeights_9:float=0
    ConstraintWeights_10:float=0.0131
    ConstraintWeights_11:float=0
   ) -> (Pose) {
    alphaBoolBlend {blendOption=Linear}
    boneToModify {boneName=dyn_L_leaf_jnt_6}
    constraintSetup [
     {targetBone {boneName=clavicle_l}}
     {targetBone {boneName=upperarm_twist_02_l}}
     {targetBone {boneName=spine_02}}
     {targetBone {boneName=spine_03}}
     {targetBone {boneName=spine_04}}
     {targetBone {boneName=spine_05}}
     {targetBone {boneName=upperarm_twist_01_l}}
     {targetBone {boneName=clavicle_r}}
     {targetBone {boneName=lowerarm_l}}
     {targetBone {boneName=lowerarm_twist_01_l}}
     {targetBone {boneName=upperarm_l}}
     {targetBone {boneName=lowerarm_twist_02_l}}
    ]
    constraintWeights [1,1,1,1,1,1,1,1,1,1,1,1]
   }
   n24 = "Constraint - dyn_L_leaf_jnt_7 (PPPPPPPPPPPP)"!(
    ComponentPose=@n23
    Alpha=1
    ConstraintWeights_0=0.1482
    ConstraintWeights_1:float=0.0085
    ConstraintWeights_2:float=0.0021
    ConstraintWeights_3:float=0.0137
    ConstraintWeights_4:float=0.0154
    ConstraintWeights_5:float=0.0147
    ConstraintWeights_6:float=0.7826
    ConstraintWeights_7:float=0.0025
    ConstraintWeights_8:float=0
    ConstraintWeights_9:float=0
    ConstraintWeights_10:float=0.0121
    ConstraintWeights_11:float=0
   ) -> (Pose) {
    alphaBoolBlend {blendOption=Linear}
    boneToModify {boneName=dyn_L_leaf_jnt_7}
    constraintSetup [
     {targetBone {boneName=clavicle_l}}
     {targetBone {boneName=upperarm_twist_02_l}}
     {targetBone {boneName=spine_02}}
     {targetBone {boneName=spine_03}}
     {targetBone {boneName=spine_04}}
     {targetBone {boneName=spine_05}}
     {targetBone {boneName=upperarm_twist_01_l}}
     {targetBone {boneName=clavicle_r}}
     {targetBone {boneName=lowerarm_l}}
     {targetBone {boneName=lowerarm_twist_01_l}}
     {targetBone {boneName=upperarm_l}}
     {targetBone {boneName=lowerarm_twist_02_l}}
    ]
    constraintWeights [1,1,1,1,1,1,1,1,1,1,1,1]
   }
   n25 = "Constraint - dyn_L_leaf_jnt_8 (PPPPPPPPPPPP)"!(
    ComponentPose=@n24
    Alpha=1
    ConstraintWeights_0=0.1535
    ConstraintWeights_1:float=0.0086
    ConstraintWeights_2:float=0.0026
    ConstraintWeights_3:float=0.0194
    ConstraintWeights_4:float=0.0227
    ConstraintWeights_5:float=0.0205
    ConstraintWeights_6:float=0.7513
    ConstraintWeights_7:float=0.0038
    ConstraintWeights_8:float=0
    ConstraintWeights_9:float=0
    ConstraintWeights_10:float=0.0175
    ConstraintWeights_11:float=0
   ) -> (Pose) {
    alphaBoolBlend {blendOption=Linear}
    boneToModify {boneName=dyn_L_leaf_jnt_8}
    constraintSetup [
     {targetBone {boneName=clavicle_l}}
     {targetBone {boneName=upperarm_twist_02_l}}
     {targetBone {boneName=spine_02}}
     {targetBone {boneName=spine_03}}
     {targetBone {boneName=spine_04}}
     {targetBone {boneName=spine_05}}
     {targetBone {boneName=upperarm_twist_01_l}}
     {targetBone {boneName=clavicle_r}}
     {targetBone {boneName=lowerarm_l}}
     {targetBone {boneName=lowerarm_twist_01_l}}
     {targetBone {boneName=upperarm_l}}
     {targetBone {boneName=lowerarm_twist_02_l}}
    ]
    constraintWeights [1,1,1,1,1,1,1,1,1,1,1,1]
   }
   n26 = "Constraint - dyn_L_leaf_jnt_9 (PPPPPPPPPPPP)"!(
    ComponentPose=@n25
    Alpha=1
    ConstraintWeights_0=0.3577
    ConstraintWeights_1:float=0.0004
    ConstraintWeights_2:float=0.0062
    ConstraintWeights_3:float=0.1352
    ConstraintWeights_4:float=0.1154
    ConstraintWeights_5:float=0.2818
    ConstraintWeights_6:float=0.0556
    ConstraintWeights_7:float=0.0021
    ConstraintWeights_8:float=0
    ConstraintWeights_9:float=0
    ConstraintWeights_10:float=0.0456
    ConstraintWeights_11:float=0
   ) -> (Pose) {
    alphaBoolBlend {blendOption=Linear}
    boneToModify {boneName=dyn_L_leaf_jnt_9}
    constraintSetup [
     {targetBone {boneName=clavicle_l}}
     {targetBone {boneName=upperarm_twist_02_l}}
     {targetBone {boneName=spine_02}}
     {targetBone {boneName=spine_03}}
     {targetBone {boneName=spine_04}}
     {targetBone {boneName=spine_05}}
     {targetBone {boneName=upperarm_twist_01_l}}
     {targetBone {boneName=clavicle_r}}
     {targetBone {boneName=lowerarm_l}}
     {targetBone {boneName=lowerarm_twist_01_l}}
     {targetBone {boneName=upperarm_l}}
     {targetBone {boneName=lowerarm_twist_02_l}}
    ]
    constraintWeights [1,1,1,1,1,1,1,1,1,1,1,1]
   }
   n27 = "Constraint - dyn_L_leaf_jnt_10 (PPPPPPPPPPPP)"!(
    ComponentPose=@n26
    Alpha=1
    ConstraintWeights_0=0.3134
    ConstraintWeights_1:float=0.0054
    ConstraintWeights_2:float=0.0054
    ConstraintWeights_3:float=0.0532
    ConstraintWeights_4:float=0.0552
    ConstraintWeights_5:float=0.0777
    ConstraintWeights_6:float=0.4481
    ConstraintWeights_7:float=0.0067
    ConstraintWeights_8:float=0
    ConstraintWeights_9:float=0
    ConstraintWeights_10:float=0.035
    ConstraintWeights_11:float=0
   ) -> (Pose) {
    alphaBoolBlend {blendOption=Linear}
    boneToModify {boneName=dyn_L_leaf_jnt_10}
    constraintSetup [
     {targetBone {boneName=clavicle_l}}
     {targetBone {boneName=upperarm_twist_02_l}}
     {targetBone {boneName=spine_02}}
     {targetBone {boneName=spine_03}}
     {targetBone {boneName=spine_04}}
     {targetBone {boneName=spine_05}}
     {targetBone {boneName=upperarm_twist_01_l}}
     {targetBone {boneName=clavicle_r}}
     {targetBone {boneName=lowerarm_l}}
     {targetBone {boneName=lowerarm_twist_01_l}}
     {targetBone {boneName=upperarm_l}}
     {targetBone {boneName=lowerarm_twist_02_l}}
    ]
    constraintWeights [1,1,1,1,1,1,1,1,1,1,1,1]
   }
   n28 = "Constraint - dyn_L_leaf_jnt_11 (PPPPPPPPPPPP)"!(
    ComponentPose=@n27
    Alpha=1
    ConstraintWeights_0=0.221
    ConstraintWeights_1:float=0.013
    ConstraintWeights_2:float=0.0136
    ConstraintWeights_3:float=0.0903
    ConstraintWeights_4:float=0.1007
    ConstraintWeights_5:float=0.0796
    ConstraintWeights_6:float=0.4128
    ConstraintWeights_7:float=0.0163
    ConstraintWeights_8:float=0
    ConstraintWeights_9:float=0
    ConstraintWeights_10:float=0.0529
    ConstraintWeights_11:float=0
   ) -> (Pose) {
    alphaBoolBlend {blendOption=Linear}
    boneToModify {boneName=dyn_L_leaf_jnt_11}
    constraintSetup [
     {targetBone {boneName=clavicle_l}}
     {targetBone {boneName=upperarm_twist_02_l}}
     {targetBone {boneName=spine_02}}
     {targetBone {boneName=spine_03}}
     {targetBone {boneName=spine_04}}
     {targetBone {boneName=spine_05}}
     {targetBone {boneName=upperarm_twist_01_l}}
     {targetBone {boneName=clavicle_r}}
     {targetBone {boneName=lowerarm_l}}
     {targetBone {boneName=lowerarm_twist_01_l}}
     {targetBone {boneName=upperarm_l}}
     {targetBone {boneName=lowerarm_twist_02_l}}
    ]
    constraintWeights [1,1,1,1,1,1,1,1,1,1,1,1]
   }
   n29 = "Constraint - dyn_L_leaf_jnt_12 (PPPPPPPPPPPP)"!(
    ComponentPose=@n28
    Alpha=1
    ConstraintWeights_0=0.0867
    ConstraintWeights_1:float=0.0217
    ConstraintWeights_2:float=0.003
    ConstraintWeights_3:float=0.0181
    ConstraintWeights_4:float=0.0241
    ConstraintWeights_5:float=0.0131
    ConstraintWeights_6:float=0.8097
    ConstraintWeights_7:float=0.0051
    ConstraintWeights_8:float=0
    ConstraintWeights_9:float=0
    ConstraintWeights_10:float=0.0185
    ConstraintWeights_11:float=0
   ) -> (Pose) {
    alphaBoolBlend {blendOption=Linear}
    boneToModify {boneName=dyn_L_leaf_jnt_12}
    constraintSetup [
     {targetBone {boneName=clavicle_l}}
     {targetBone {boneName=upperarm_twist_02_l}}
     {targetBone {boneName=spine_02}}
     {targetBone {boneName=spine_03}}
     {targetBone {boneName=spine_04}}
     {targetBone {boneName=spine_05}}
     {targetBone {boneName=upperarm_twist_01_l}}
     {targetBone {boneName=clavicle_r}}
     {targetBone {boneName=lowerarm_l}}
     {targetBone {boneName=lowerarm_twist_01_l}}
     {targetBone {boneName=upperarm_l}}
     {targetBone {boneName=lowerarm_twist_02_l}}
    ]
    constraintWeights [1,1,1,1,1,1,1,1,1,1,1,1]
   }
   n30 = "Constraint - dyn_L_leaf_jnt_13 (PPPPPPPPPPPP)"!(
    ComponentPose=@n29
    Alpha=1
    ConstraintWeights_0=0.3038
    ConstraintWeights_1:float=0.0045
    ConstraintWeights_2:float=0.0129
    ConstraintWeights_3:float=0.1242
    ConstraintWeights_4:float=0.1329
    ConstraintWeights_5:float=0.1312
    ConstraintWeights_6:float=0.2126
    ConstraintWeights_7:float=0.013
    ConstraintWeights_8:float=0
    ConstraintWeights_9:float=0
    ConstraintWeights_10:float=0.0649
    ConstraintWeights_11:float=0
   ) -> (Pose) {
    alphaBoolBlend {blendOption=Linear}
    boneToModify {boneName=dyn_L_leaf_jnt_13}
    constraintSetup [
     {targetBone {boneName=clavicle_l}}
     {targetBone {boneName=upperarm_twist_02_l}}
     {targetBone {boneName=spine_02}}
     {targetBone {boneName=spine_03}}
     {targetBone {boneName=spine_04}}
     {targetBone {boneName=spine_05}}
     {targetBone {boneName=upperarm_twist_01_l}}
     {targetBone {boneName=clavicle_r}}
     {targetBone {boneName=lowerarm_l}}
     {targetBone {boneName=lowerarm_twist_01_l}}
     {targetBone {boneName=upperarm_l}}
     {targetBone {boneName=lowerarm_twist_02_l}}
    ]
    constraintWeights [1,1,1,1,1,1,1,1,1,1,1,1]
   }
   n31 = "Constraint - dyn_L_leaf_jnt_14 (PPPPPPPPPPPP)"!(
    ComponentPose=@n30
    Alpha=1
    ConstraintWeights_0=0.307
    ConstraintWeights_1:float=0.0058
    ConstraintWeights_2:float=0.0096
    ConstraintWeights_3:float=0.0981
    ConstraintWeights_4:float=0.1018
    ConstraintWeights_5:float=0.1269
    ConstraintWeights_6:float=0.2845
    ConstraintWeights_7:float=0.0106
    ConstraintWeights_8:float=0
    ConstraintWeights_9:float=0
    ConstraintWeights_10:float=0.0558
    ConstraintWeights_11:float=0
   ) -> (Pose) {
    alphaBoolBlend {blendOption=Linear}
    boneToModify {boneName=dyn_L_leaf_jnt_14}
    constraintSetup [
     {targetBone {boneName=clavicle_l}}
     {targetBone {boneName=upperarm_twist_02_l}}
     {targetBone {boneName=spine_02}}
     {targetBone {boneName=spine_03}}
     {targetBone {boneName=spine_04}}
     {targetBone {boneName=spine_05}}
     {targetBone {boneName=upperarm_twist_01_l}}
     {targetBone {boneName=clavicle_r}}
     {targetBone {boneName=lowerarm_l}}
     {targetBone {boneName=lowerarm_twist_01_l}}
     {targetBone {boneName=upperarm_l}}
     {targetBone {boneName=lowerarm_twist_02_l}}
    ]
    constraintWeights [1,1,1,1,1,1,1,1,1,1,1,1]
   }
   n32 = "Constraint - dyn_L_leaf_jnt_15 (PPPPPPPPPPPP)"!(
    ComponentPose=@n31
    Alpha=1
    ConstraintWeights_0=0.1298
    ConstraintWeights_1:float=0.0176
    ConstraintWeights_2:float=0.0068
    ConstraintWeights_3:float=0.0424
    ConstraintWeights_4:float=0.0518
    ConstraintWeights_5:float=0.0323
    ConstraintWeights_6:float=0.6752
    ConstraintWeights_7:float=0.0103
    ConstraintWeights_8:float=0
    ConstraintWeights_9:float=0
    ConstraintWeights_10:float=0.0338
    ConstraintWeights_11:float=0
   ) -> (Pose) {
    alphaBoolBlend {blendOption=Linear}
    boneToModify {boneName=dyn_L_leaf_jnt_15}
    constraintSetup [
     {targetBone {boneName=clavicle_l}}
     {targetBone {boneName=upperarm_twist_02_l}}
     {targetBone {boneName=spine_02}}
     {targetBone {boneName=spine_03}}
     {targetBone {boneName=spine_04}}
     {targetBone {boneName=spine_05}}
     {targetBone {boneName=upperarm_twist_01_l}}
     {targetBone {boneName=clavicle_r}}
     {targetBone {boneName=lowerarm_l}}
     {targetBone {boneName=lowerarm_twist_01_l}}
     {targetBone {boneName=upperarm_l}}
     {targetBone {boneName=lowerarm_twist_02_l}}
    ]
    constraintWeights [1,1,1,1,1,1,1,1,1,1,1,1]
   }
   n33 = "Constraint - dyn_L_leaf_jnt_16 (PPPPPPPPPPPP)"!(
    ComponentPose=@n32
    Alpha=1
    ConstraintWeights_0=0.5357
    ConstraintWeights_1:float=0.0011
    ConstraintWeights_2:float=0.0019
    ConstraintWeights_3:float=0.0248
    ConstraintWeights_4:float=0.0208
    ConstraintWeights_5:float=0.0711
    ConstraintWeights_6:float=0.3267
    ConstraintWeights_7:float=0.0016
    ConstraintWeights_8:float=0
    ConstraintWeights_9:float=0
    ConstraintWeights_10:float=0.0164
    ConstraintWeights_11:float=0
   ) -> (Pose) {
    alphaBoolBlend {blendOption=Linear}
    boneToModify {boneName=dyn_L_leaf_jnt_16}
    constraintSetup [
     {targetBone {boneName=clavicle_l}}
     {targetBone {boneName=upperarm_twist_02_l}}
     {targetBone {boneName=spine_02}}
     {targetBone {boneName=spine_03}}
     {targetBone {boneName=spine_04}}
     {targetBone {boneName=spine_05}}
     {targetBone {boneName=upperarm_twist_01_l}}
     {targetBone {boneName=clavicle_r}}
     {targetBone {boneName=lowerarm_l}}
     {targetBone {boneName=lowerarm_twist_01_l}}
     {targetBone {boneName=upperarm_l}}
     {targetBone {boneName=lowerarm_twist_02_l}}
    ]
    constraintWeights [1,1,1,1,1,1,1,1,1,1,1,1]
   }
   n34 = "Constraint - dyn_R_leaf_jnt_1 (PPPPPPPPPPP)"!(
    ComponentPose=@n33
    Alpha=1
    ConstraintWeights_0=0
    ConstraintWeights_1:float=0.0449
    ConstraintWeights_2:float=0
    ConstraintWeights_3:float=0
    ConstraintWeights_4:float=0.0745
    ConstraintWeights_5:float=0
    ConstraintWeights_6:float=0.0027
    ConstraintWeights_7:float=0.0372
    ConstraintWeights_8:float=0.052
    ConstraintWeights_9:float=0.1858
    ConstraintWeights_10:float=0.6029
   ) -> (Pose) {
    alphaBoolBlend {blendOption=Linear}
    boneToModify {boneName=dyn_R_leaf_jnt_1}
    constraintSetup [
     {targetBone {boneName=lowerarm_twist_02_r}}
     {targetBone {boneName=upperarm_r}}
     {targetBone {boneName=clavicle_l}}
     {targetBone {boneName=lowerarm_r}}
     {targetBone {boneName=upperarm_twist_01_r}}
     {targetBone {boneName=upperarm_twist_02_r}}
     {targetBone {boneName=spine_02}}
     {targetBone {boneName=spine_03}}
     {targetBone {boneName=spine_04}}
     {targetBone {boneName=spine_05}}
     {targetBone {boneName=clavicle_r}}
    ]
    constraintWeights [1,1,1,1,1,1,1,1,1,1,1]
   }
   n35 = "Constraint - dyn_R_leaf_jnt_2 (PPPPPPPPPPP)"!(
    ComponentPose=@n34
    Alpha=1
    ConstraintWeights_0=0
    ConstraintWeights_1:float=0.0177
    ConstraintWeights_2:float=0
    ConstraintWeights_3:float=0
    ConstraintWeights_4:float=0.0319
    ConstraintWeights_5:float=0
    ConstraintWeights_6:float=0.0003
    ConstraintWeights_7:float=0.0065
    ConstraintWeights_8:float=0.0112
    ConstraintWeights_9:float=0.2259
    ConstraintWeights_10:float=0.7064
   ) -> (Pose) {
    alphaBoolBlend {blendOption=Linear}
    boneToModify {boneName=dyn_R_leaf_jnt_2}
    constraintSetup [
     {targetBone {boneName=lowerarm_twist_02_r}}
     {targetBone {boneName=upperarm_r}}
     {targetBone {boneName=clavicle_l}}
     {targetBone {boneName=lowerarm_r}}
     {targetBone {boneName=upperarm_twist_01_r}}
     {targetBone {boneName=upperarm_twist_02_r}}
     {targetBone {boneName=spine_02}}
     {targetBone {boneName=spine_03}}
     {targetBone {boneName=spine_04}}
     {targetBone {boneName=spine_05}}
     {targetBone {boneName=clavicle_r}}
    ]
    constraintWeights [1,1,1,1,1,1,1,1,1,1,1]
   }
   n36 = "Constraint - dyn_R_leaf_jnt_3 (PPPPPPPPPPP)"!(
    ComponentPose=@n35
    Alpha=1
    ConstraintWeights_0=0
    ConstraintWeights_1:float=0.0157
    ConstraintWeights_2:float=0
    ConstraintWeights_3:float=0
    ConstraintWeights_4:float=0.1433
    ConstraintWeights_5:float=0
    ConstraintWeights_6:float=0.004
    ConstraintWeights_7:float=0.0074
    ConstraintWeights_8:float=0.0106
    ConstraintWeights_9:float=0.0498
    ConstraintWeights_10:float=0.7693
   ) -> (Pose) {
    alphaBoolBlend {blendOption=Linear}
    boneToModify {boneName=dyn_R_leaf_jnt_3}
    constraintSetup [
     {targetBone {boneName=lowerarm_twist_02_r}}
     {targetBone {boneName=upperarm_r}}
     {targetBone {boneName=clavicle_l}}
     {targetBone {boneName=lowerarm_r}}
     {targetBone {boneName=upperarm_twist_01_r}}
     {targetBone {boneName=upperarm_twist_02_r}}
     {targetBone {boneName=spine_02}}
     {targetBone {boneName=spine_03}}
     {targetBone {boneName=spine_04}}
     {targetBone {boneName=spine_05}}
     {targetBone {boneName=clavicle_r}}
    ]
    constraintWeights [1,1,1,1,1,1,1,1,1,1,1]
   }
   n37 = "Constraint - dyn_R_leaf_jnt_4 (PPPPPPPPPPP)"!(
    ComponentPose=@n36
    Alpha=1
    ConstraintWeights_0=0
    ConstraintWeights_1:float=0.0313
    ConstraintWeights_2:float=0
    ConstraintWeights_3:float=0
    ConstraintWeights_4:float=0.2944
    ConstraintWeights_5:float=0
    ConstraintWeights_6:float=0.0032
    ConstraintWeights_7:float=0.0088
    ConstraintWeights_8:float=0.0151
    ConstraintWeights_9:float=0.0558
    ConstraintWeights_10:float=0.5915
   ) -> (Pose) {
    alphaBoolBlend {blendOption=Linear}
    boneToModify {boneName=dyn_R_leaf_jnt_4}
    constraintSetup [
     {targetBone {boneName=lowerarm_twist_02_r}}
     {targetBone {boneName=upperarm_r}}
     {targetBone {boneName=clavicle_l}}
     {targetBone {boneName=lowerarm_r}}
     {targetBone {boneName=upperarm_twist_01_r}}
     {targetBone {boneName=upperarm_twist_02_r}}
     {targetBone {boneName=spine_02}}
     {targetBone {boneName=spine_03}}
     {targetBone {boneName=spine_04}}
     {targetBone {boneName=spine_05}}
     {targetBone {boneName=clavicle_r}}
    ]
    constraintWeights [1,1,1,1,1,1,1,1,1,1,1]
   }
   n38 = "Constraint - dyn_R_leaf_jnt_5 (PPPPPPPPPPP)"!(
    ComponentPose=@n37
    Alpha=1
    ConstraintWeights_0=0
    ConstraintWeights_1:float=0.0367
    ConstraintWeights_2:float=0
    ConstraintWeights_3:float=0
    ConstraintWeights_4:float=0.402
    ConstraintWeights_5:float=0
    ConstraintWeights_6:float=0.0032
    ConstraintWeights_7:float=0.0097
    ConstraintWeights_8:float=0.0174
    ConstraintWeights_9:float=0.0474
    ConstraintWeights_10:float=0.4836
   ) -> (Pose) {
    alphaBoolBlend {blendOption=Linear}
    boneToModify {boneName=dyn_R_leaf_jnt_5}
    constraintSetup [
     {targetBone {boneName=lowerarm_twist_02_r}}
     {targetBone {boneName=upperarm_r}}
     {targetBone {boneName=clavicle_l}}
     {targetBone {boneName=lowerarm_r}}
     {targetBone {boneName=upperarm_twist_01_r}}
     {targetBone {boneName=upperarm_twist_02_r}}
     {targetBone {boneName=spine_02}}
     {targetBone {boneName=spine_03}}
     {targetBone {boneName=spine_04}}
     {targetBone {boneName=spine_05}}
     {targetBone {boneName=clavicle_r}}
    ]
    constraintWeights [1,1,1,1,1,1,1,1,1,1,1]
   }
   n39 = "Constraint - dyn_R_leaf_jnt_6 (PPPPPPPPPPP)"!(
    ComponentPose=@n38
    Alpha=1
    ConstraintWeights_0=0
    ConstraintWeights_1:float=0.0374
    ConstraintWeights_2:float=0
    ConstraintWeights_3:float=0
    ConstraintWeights_4:float=0.5228
    ConstraintWeights_5:float=0
    ConstraintWeights_6:float=0.0102
    ConstraintWeights_7:float=0.0221
    ConstraintWeights_8:float=0.0331
    ConstraintWeights_9:float=0.04
    ConstraintWeights_10:float=0.3344
   ) -> (Pose) {
    alphaBoolBlend {blendOption=Linear}
    boneToModify {boneName=dyn_R_leaf_jnt_6}
    constraintSetup [
     {targetBone {boneName=lowerarm_twist_02_r}}
     {targetBone {boneName=upperarm_r}}
     {targetBone {boneName=clavicle_l}}
     {targetBone {boneName=lowerarm_r}}
     {targetBone {boneName=upperarm_twist_01_r}}
     {targetBone {boneName=upperarm_twist_02_r}}
     {targetBone {boneName=spine_02}}
     {targetBone {boneName=spine_03}}
     {targetBone {boneName=spine_04}}
     {targetBone {boneName=spine_05}}
     {targetBone {boneName=clavicle_r}}
    ]
    constraintWeights [1,1,1,1,1,1,1,1,1,1,1]
   }
   n40 = "Constraint - dyn_R_leaf_jnt_7 (PPPPPPPPPPP)"!(
    ComponentPose=@n39
    Alpha=1
    ConstraintWeights_0=0
    ConstraintWeights_1:float=0.0356
    ConstraintWeights_2:float=0
    ConstraintWeights_3:float=0
    ConstraintWeights_4:float=0.6556
    ConstraintWeights_5:float=0
    ConstraintWeights_6:float=0.0029
    ConstraintWeights_7:float=0.0093
    ConstraintWeights_8:float=0.0178
    ConstraintWeights_9:float=0.0222
    ConstraintWeights_10:float=0.2564
   ) -> (Pose) {
    alphaBoolBlend {blendOption=Linear}
    boneToModify {boneName=dyn_R_leaf_jnt_7}
    constraintSetup [
     {targetBone {boneName=lowerarm_twist_02_r}}
     {targetBone {boneName=upperarm_r}}
     {targetBone {boneName=clavicle_l}}
     {targetBone {boneName=lowerarm_r}}
     {targetBone {boneName=upperarm_twist_01_r}}
     {targetBone {boneName=upperarm_twist_02_r}}
     {targetBone {boneName=spine_02}}
     {targetBone {boneName=spine_03}}
     {targetBone {boneName=spine_04}}
     {targetBone {boneName=spine_05}}
     {targetBone {boneName=clavicle_r}}
    ]
    constraintWeights [1,1,1,1,1,1,1,1,1,1,1]
   }
   n41 = "Constraint - dyn_R_leaf_jnt_8 (PPPPPPPPPPP)"!(
    ComponentPose=@n40
    Alpha=1
    ConstraintWeights_0=0
    ConstraintWeights_1:float=0.0386
    ConstraintWeights_2:float=0
    ConstraintWeights_3:float=0
    ConstraintWeights_4:float=0.3973
    ConstraintWeights_5:float=0
    ConstraintWeights_6:float=0.0022
    ConstraintWeights_7:float=0.0101
    ConstraintWeights_8:float=0.0177
    ConstraintWeights_9:float=0.0482
    ConstraintWeights_10:float=0.4859
   ) -> (Pose) {
    alphaBoolBlend {blendOption=Linear}
    boneToModify {boneName=dyn_R_leaf_jnt_8}
    constraintSetup [
     {targetBone {boneName=lowerarm_twist_02_r}}
     {targetBone {boneName=upperarm_r}}
     {targetBone {boneName=clavicle_l}}
     {targetBone {boneName=lowerarm_r}}
     {targetBone {boneName=upperarm_twist_01_r}}
     {targetBone {boneName=upperarm_twist_02_r}}
     {targetBone {boneName=spine_02}}
     {targetBone {boneName=spine_03}}
     {targetBone {boneName=spine_04}}
     {targetBone {boneName=spine_05}}
     {targetBone {boneName=clavicle_r}}
    ]
    constraintWeights [1,1,1,1,1,1,1,1,1,1,1]
   }
   n42 = "Constraint - dyn_R_leaf_jnt_9 (PPPPPPPPPPP)"!(
    ComponentPose=@n41
    Alpha=1
    ConstraintWeights_0=0
    ConstraintWeights_1:float=0.0311
    ConstraintWeights_2:float=0
    ConstraintWeights_3:float=0
    ConstraintWeights_4:float=0.7398
    ConstraintWeights_5:float=0
    ConstraintWeights_6:float=0.0021
    ConstraintWeights_7:float=0.0089
    ConstraintWeights_8:float=0.0165
    ConstraintWeights_9:float=0.0157
    ConstraintWeights_10:float=0.1859
   ) -> (Pose) {
    alphaBoolBlend {blendOption=Linear}
    boneToModify {boneName=dyn_R_leaf_jnt_9}
    constraintSetup [
     {targetBone {boneName=lowerarm_twist_02_r}}
     {targetBone {boneName=upperarm_r}}
     {targetBone {boneName=clavicle_l}}
     {targetBone {boneName=lowerarm_r}}
     {targetBone {boneName=upperarm_twist_01_r}}
     {targetBone {boneName=upperarm_twist_02_r}}
     {targetBone {boneName=spine_02}}
     {targetBone {boneName=spine_03}}
     {targetBone {boneName=spine_04}}
     {targetBone {boneName=spine_05}}
     {targetBone {boneName=clavicle_r}}
    ]
    constraintWeights [1,1,1,1,1,1,1,1,1,1,1]
   }
   n43 = "Constraint - dyn_R_leaf_jnt_10 (PPPPPPPPPPP)"!(
    ComponentPose=@n42
    Alpha=1
    ConstraintWeights_0=0
    ConstraintWeights_1:float=0.0422
    ConstraintWeights_2:float=0
    ConstraintWeights_3:float=0
    ConstraintWeights_4:float=0.6345
    ConstraintWeights_5:float=0
    ConstraintWeights_6:float=0.002
    ConstraintWeights_7:float=0.0123
    ConstraintWeights_8:float=0.0209
    ConstraintWeights_9:float=0.0235
    ConstraintWeights_10:float=0.2646
   ) -> (Pose) {
    alphaBoolBlend {blendOption=Linear}
    boneToModify {boneName=dyn_R_leaf_jnt_10}
    constraintSetup [
     {targetBone {boneName=lowerarm_twist_02_r}}
     {targetBone {boneName=upperarm_r}}
     {targetBone {boneName=clavicle_l}}
     {targetBone {boneName=lowerarm_r}}
     {targetBone {boneName=upperarm_twist_01_r}}
     {targetBone {boneName=upperarm_twist_02_r}}
     {targetBone {boneName=spine_02}}
     {targetBone {boneName=spine_03}}
     {targetBone {boneName=spine_04}}
     {targetBone {boneName=spine_05}}
     {targetBone {boneName=clavicle_r}}
    ]
    constraintWeights [1,1,1,1,1,1,1,1,1,1,1]
   }
   n44 = "Constraint - dyn_R_leaf_jnt_11 (PPPPPPPPPPP)"!(
    ComponentPose=@n43
    Alpha=1
    ConstraintWeights_0=0
    ConstraintWeights_1:float=0.054
    ConstraintWeights_2:float=0
    ConstraintWeights_3:float=0
    ConstraintWeights_4:float=0.5378
    ConstraintWeights_5:float=0
    ConstraintWeights_6:float=0.0024
    ConstraintWeights_7:float=0.0183
    ConstraintWeights_8:float=0.0286
    ConstraintWeights_9:float=0.0333
    ConstraintWeights_10:float=0.3256
   ) -> (Pose) {
    alphaBoolBlend {blendOption=Linear}
    boneToModify {boneName=dyn_R_leaf_jnt_11}
    constraintSetup [
     {targetBone {boneName=lowerarm_twist_02_r}}
     {targetBone {boneName=upperarm_r}}
     {targetBone {boneName=clavicle_l}}
     {targetBone {boneName=lowerarm_r}}
     {targetBone {boneName=upperarm_twist_01_r}}
     {targetBone {boneName=upperarm_twist_02_r}}
     {targetBone {boneName=spine_02}}
     {targetBone {boneName=spine_03}}
     {targetBone {boneName=spine_04}}
     {targetBone {boneName=spine_05}}
     {targetBone {boneName=clavicle_r}}
    ]
    constraintWeights [1,1,1,1,1,1,1,1,1,1,1]
   }
   n45 = "Constraint - dyn_R_leaf_jnt_12 (PPPPPPPPPPP)"!(
    ComponentPose=@n44
    Alpha=0.5
    ConstraintWeights_0=0
    ConstraintWeights_1:float=0.0144
    ConstraintWeights_2:float=0
    ConstraintWeights_3:float=0
    ConstraintWeights_4:float=0.8967
    ConstraintWeights_5:float=0
    ConstraintWeights_6:float=0.0011
    ConstraintWeights_7:float=0.005
    ConstraintWeights_8:float=0.0103
    ConstraintWeights_9:float=0.0052
    ConstraintWeights_10:float=0.0673
   ) -> (Pose) {
    alphaBoolBlend {blendOption=Linear}
    boneToModify {boneName=dyn_R_leaf_jnt_12}
    constraintSetup [
     {targetBone {boneName=lowerarm_twist_02_r}}
     {targetBone {boneName=upperarm_r}}
     {targetBone {boneName=clavicle_l}}
     {targetBone {boneName=lowerarm_r}}
     {targetBone {boneName=upperarm_twist_01_r}}
     {targetBone {boneName=upperarm_twist_02_r}}
     {targetBone {boneName=spine_02}}
     {targetBone {boneName=spine_03}}
     {targetBone {boneName=spine_04}}
     {targetBone {boneName=spine_05}}
     {targetBone {boneName=clavicle_r}}
    ]
    constraintWeights [1,1,1,1,1,1,1,1,1,1,1]
   }
   n46 = "Constraint - dyn_R_leaf_jnt_13 (PPPPPPPPPPP)"!(
    ComponentPose=@n45
    Alpha=1
    ConstraintWeights_0=0
    ConstraintWeights_1:float=0.0238
    ConstraintWeights_2:float=0
    ConstraintWeights_3:float=0
    ConstraintWeights_4:float=0.8439
    ConstraintWeights_5:float=0
    ConstraintWeights_6:float=0.0033
    ConstraintWeights_7:float=0.0114
    ConstraintWeights_8:float=0.0222
    ConstraintWeights_9:float=0.0129
    ConstraintWeights_10:float=0.0826
   ) -> (Pose) {
    alphaBoolBlend {blendOption=Linear}
    boneToModify {boneName=dyn_R_leaf_jnt_13}
    constraintSetup [
     {targetBone {boneName=lowerarm_twist_02_r}}
     {targetBone {boneName=upperarm_r}}
     {targetBone {boneName=clavicle_l}}
     {targetBone {boneName=lowerarm_r}}
     {targetBone {boneName=upperarm_twist_01_r}}
     {targetBone {boneName=upperarm_twist_02_r}}
     {targetBone {boneName=spine_02}}
     {targetBone {boneName=spine_03}}
     {targetBone {boneName=spine_04}}
     {targetBone {boneName=spine_05}}
     {targetBone {boneName=clavicle_r}}
    ]
    constraintWeights [1,1,1,1,1,1,1,1,1,1,1]
   }
   n47 = "Constraint - dyn_R_leaf_jnt_14 (PPPPPPPPPPP)"!(
    ComponentPose=@n46
    Alpha=1
    ConstraintWeights_0=0
    ConstraintWeights_1:float=0.0009
    ConstraintWeights_2:float=0
    ConstraintWeights_3:float=0
    ConstraintWeights_4:float=0.9974
    ConstraintWeights_5:float=0.0006
    ConstraintWeights_6:float=0
    ConstraintWeights_7:float=0.0004
    ConstraintWeights_8:float=0.0007
    ConstraintWeights_9:float=0
    ConstraintWeights_10:float=0.0001
   ) -> (Pose) {
    alphaBoolBlend {blendOption=Linear}
    boneToModify {boneName=dyn_R_leaf_jnt_14}
    constraintSetup [
     {targetBone {boneName=lowerarm_twist_02_r}}
     {targetBone {boneName=upperarm_r}}
     {targetBone {boneName=clavicle_l}}
     {targetBone {boneName=lowerarm_r}}
     {targetBone {boneName=upperarm_twist_01_r}}
     {targetBone {boneName=upperarm_twist_02_r}}
     {targetBone {boneName=spine_02}}
     {targetBone {boneName=spine_03}}
     {targetBone {boneName=spine_04}}
     {targetBone {boneName=spine_05}}
     {targetBone {boneName=clavicle_r}}
    ]
    constraintWeights [1,1,1,1,1,1,1,1,1,1,1]
   }
   n48 = "Constraint - dyn_R_leaf_jnt_15 (PPPPPPPPPPP)"!(
    ComponentPose=@n47
    Alpha=1
    ConstraintWeights_0=0
    ConstraintWeights_1:float=0.0306
    ConstraintWeights_2:float=0
    ConstraintWeights_3:float=0
    ConstraintWeights_4:float=0.8591
    ConstraintWeights_5:float=0.0001
    ConstraintWeights_6:float=0.0006
    ConstraintWeights_7:float=0.008
    ConstraintWeights_8:float=0.0143
    ConstraintWeights_9:float=0.0039
    ConstraintWeights_10:float=0.0833
   ) -> (Pose) {
    alphaBoolBlend {blendOption=Linear}
    boneToModify {boneName=dyn_R_leaf_jnt_15}
    constraintSetup [
     {targetBone {boneName=lowerarm_twist_02_r}}
     {targetBone {boneName=upperarm_r}}
     {targetBone {boneName=clavicle_l}}
     {targetBone {boneName=lowerarm_r}}
     {targetBone {boneName=upperarm_twist_01_r}}
     {targetBone {boneName=upperarm_twist_02_r}}
     {targetBone {boneName=spine_02}}
     {targetBone {boneName=spine_03}}
     {targetBone {boneName=spine_04}}
     {targetBone {boneName=spine_05}}
     {targetBone {boneName=clavicle_r}}
    ]
    constraintWeights [1,1,1,1,1,1,1,1,1,1,1]
   }
   n49 = "Constraint - dyn_R_leaf_jnt_16 (PPPPPPPPPPP)"!(
    ComponentPose=@n48
    Alpha=1
    ConstraintWeights_0=0
    ConstraintWeights_1:float=0.0642
    ConstraintWeights_2:float=0
    ConstraintWeights_3:float=0
    ConstraintWeights_4:float=0.7544
    ConstraintWeights_5:float=0.0001
    ConstraintWeights_6:float=0.002
    ConstraintWeights_7:float=0.0235
    ConstraintWeights_8:float=0.0372
    ConstraintWeights_9:float=0.006
    ConstraintWeights_10:float=0.1126
   ) -> (Pose) {
    alphaBoolBlend {blendOption=Linear}
    boneToModify {boneName=dyn_R_leaf_jnt_16}
    constraintSetup [
     {targetBone {boneName=lowerarm_twist_02_r}}
     {targetBone {boneName=upperarm_r}}
     {targetBone {boneName=clavicle_l}}
     {targetBone {boneName=lowerarm_r}}
     {targetBone {boneName=upperarm_twist_01_r}}
     {targetBone {boneName=upperarm_twist_02_r}}
     {targetBone {boneName=spine_02}}
     {targetBone {boneName=spine_03}}
     {targetBone {boneName=spine_04}}
     {targetBone {boneName=spine_05}}
     {targetBone {boneName=clavicle_r}}
    ]
    constraintWeights [1,1,1,1,1,1,1,1,1,1,1]
   }
   n50 = "Constraint - dyn_R_leaf_jnt_17 (PPPPPPPPPPP)"!(
    ComponentPose=@n49
    Alpha=1
    ConstraintWeights_0=0
    ConstraintWeights_1:float=0.063
    ConstraintWeights_2:float=0
    ConstraintWeights_3:float=0
    ConstraintWeights_4:float=0.7916
    ConstraintWeights_5:float=0.0002
    ConstraintWeights_6:float=0.0022
    ConstraintWeights_7:float=0.027
    ConstraintWeights_8:float=0.0419
    ConstraintWeights_9:float=0.0045
    ConstraintWeights_10:float=0.0696
   ) -> (Pose) {
    alphaBoolBlend {blendOption=Linear}
    boneToModify {boneName=dyn_R_leaf_jnt_17}
    constraintSetup [
     {targetBone {boneName=lowerarm_twist_02_r}}
     {targetBone {boneName=upperarm_r}}
     {targetBone {boneName=clavicle_l}}
     {targetBone {boneName=lowerarm_r}}
     {targetBone {boneName=upperarm_twist_01_r}}
     {targetBone {boneName=upperarm_twist_02_r}}
     {targetBone {boneName=spine_02}}
     {targetBone {boneName=spine_03}}
     {targetBone {boneName=spine_04}}
     {targetBone {boneName=spine_05}}
     {targetBone {boneName=clavicle_r}}
    ]
    constraintWeights [1,1,1,1,1,1,1,1,1,1,1]
   }
   n51 = "Constraint - dyn_R_leaf_jnt_18 (PPPPPPPPPPP)"!(
    ComponentPose=@n50
    Alpha=1
    ConstraintWeights_0=0
    ConstraintWeights_1:float=0.0743
    ConstraintWeights_2:float=0
    ConstraintWeights_3:float=0
    ConstraintWeights_4:float=0.2711
    ConstraintWeights_5:float=0
    ConstraintWeights_6:float=0.0082
    ConstraintWeights_7:float=0.0569
    ConstraintWeights_8:float=0.0715
    ConstraintWeights_9:float=0.0672
    ConstraintWeights_10:float=0.4508
   ) -> (Pose) {
    alphaBoolBlend {blendOption=Linear}
    boneToModify {boneName=dyn_R_leaf_jnt_18}
    constraintSetup [
     {targetBone {boneName=lowerarm_twist_02_r}}
     {targetBone {boneName=upperarm_r}}
     {targetBone {boneName=clavicle_l}}
     {targetBone {boneName=lowerarm_r}}
     {targetBone {boneName=upperarm_twist_01_r}}
     {targetBone {boneName=upperarm_twist_02_r}}
     {targetBone {boneName=spine_02}}
     {targetBone {boneName=spine_03}}
     {targetBone {boneName=spine_04}}
     {targetBone {boneName=spine_05}}
     {targetBone {boneName=clavicle_r}}
    ]
    constraintWeights [1,1,1,1,1,1,1,1,1,1,1]
   }
   n52 = "Constraint - dyn_R_leaf_jnt_19 (PPPPPPPPPPP)"!(
    ComponentPose=@n51
    Alpha=1
    ConstraintWeights_0=0
    ConstraintWeights_1:float=0.0724
    ConstraintWeights_2:float=0
    ConstraintWeights_3:float=0
    ConstraintWeights_4:float=0.3545
    ConstraintWeights_5:float=0
    ConstraintWeights_6:float=0.0047
    ConstraintWeights_7:float=0.037
    ConstraintWeights_8:float=0.0508
    ConstraintWeights_9:float=0.0561
    ConstraintWeights_10:float=0.4245
   ) -> (Pose) {
    alphaBoolBlend {blendOption=Linear}
    boneToModify {boneName=dyn_R_leaf_jnt_19}
    constraintSetup [
     {targetBone {boneName=lowerarm_twist_02_r}}
     {targetBone {boneName=upperarm_r}}
     {targetBone {boneName=clavicle_l}}
     {targetBone {boneName=lowerarm_r}}
     {targetBone {boneName=upperarm_twist_01_r}}
     {targetBone {boneName=upperarm_twist_02_r}}
     {targetBone {boneName=spine_02}}
     {targetBone {boneName=spine_03}}
     {targetBone {boneName=spine_04}}
     {targetBone {boneName=spine_05}}
     {targetBone {boneName=clavicle_r}}
    ]
    constraintWeights [1,1,1,1,1,1,1,1,1,1,1]
   }
   n53 = "Constraint - dyn_R_leaf_jnt_20 (PPPPPPPPPPP)"!(
    ComponentPose=@n52
    Alpha=1
    ConstraintWeights_0=0
    ConstraintWeights_1:float=0.0365
    ConstraintWeights_2:float=0
    ConstraintWeights_3:float=0
    ConstraintWeights_4:float=0.2109
    ConstraintWeights_5:float=0
    ConstraintWeights_6:float=0.0017
    ConstraintWeights_7:float=0.0134
    ConstraintWeights_8:float=0.021
    ConstraintWeights_9:float=0.0996
    ConstraintWeights_10:float=0.6168
   ) -> (Pose) {
    alphaBoolBlend {blendOption=Linear}
    boneToModify {boneName=dyn_R_leaf_jnt_20}
    constraintSetup [
     {targetBone {boneName=lowerarm_twist_02_r}}
     {targetBone {boneName=upperarm_r}}
     {targetBone {boneName=clavicle_l}}
     {targetBone {boneName=lowerarm_r}}
     {targetBone {boneName=upperarm_twist_01_r}}
     {targetBone {boneName=upperarm_twist_02_r}}
     {targetBone {boneName=spine_02}}
     {targetBone {boneName=spine_03}}
     {targetBone {boneName=spine_04}}
     {targetBone {boneName=spine_05}}
     {targetBone {boneName=clavicle_r}}
    ]
    constraintWeights [1,1,1,1,1,1,1,1,1,1,1]
   }
   n54 = "To Float (Boolean)"(InBool=Get bIsSkyDiving())
   n55 = "Transform (Modify) Bone - Bone: dyn_hipleaf_jnt_6"(ComponentPose=@n53, Alpha=@n54, Translation="0,0,0", Rotation="25,25,0", Scale="1,1,1") {rotationMode=BMM_Additive, alphaBoolBlend {blendOption=Linear}, boneToModify {boneName=dyn_hipleaf_jnt_6}, rotation {pitch=25, yaw=25}}
   n56 = "Transform (Modify) Bone - Bone: dyn_hipleaf_jnt_7"(ComponentPose=@n55, Alpha=@n54, Translation="0,0,0", Rotation="25,25,0", Scale="1,1,1") {rotationMode=BMM_Additive, alphaBoolBlend {blendOption=Linear}, boneToModify {boneName=dyn_hipleaf_jnt_7}, rotation {pitch=25, yaw=25}}
   n57 = "Transform (Modify) Bone - Bone: vines_correct_R_jnt"(ComponentPose=@n56, Alpha=@n54, Translation="0.236,-4.359,1.346", Rotation="0,0,0", Scale="1,1,1") {
    translationMode=BMM_Additive
    alphaBoolBlend {blendOption=Linear}
    boneToModify {boneName=vines_correct_R_jnt}
    translation=<0.236 -4.359 1.346>
   }
   Output Pose(Result=Component To Local(ComponentPose=@n57)) {layerGroup=None}
  ]}
 EventGraph {nodes [
   Event BlueprintUpdateAnimation()
   > n1 = Gravity Override State Function DEPRECATED(CharacterPartAnimInstance=Self-Reference(), GameTimeInSeconds=GetGameTimeInSeconds(), GravityOverrideParams=Get Gravity Override Params())
   > Set GravityOverride(GravityOverride=@n1)
  ]}
]
variables [
 Gravity Override Params {type=GravityOverrideParamsStruct}
 GravityOverride {type=Vector}
]
