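# Physics asset for the F_MED_Moss_Bomber female player body, named as the collider set for cape simulation.
# Capsules cover the head, clavicles, pelvis, spine and feet, with paired capsules on spine_02 to 04, while arm, thigh, calf and neck bodies carry no shapes.
# Dynamic shoulder-leaf, thigh-leaf and knot joints also have capsules.
# Only the shoulder leaves are constrained, chained to spine_05 with free swing and twist.
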PhysicsAsset /BRCosmetics/Characters/Player/Female/Medium/Bodies/F_MED_Moss_Bomber/Meshes/F_MED_Moss_Bomber_CapeColliders
bodies {
 calf_l []
 calf_r []
 clavicle_l [{shapeType=Capsule, length=4.6964, radius=10.6353, center=<8.8039 -0.4422 -1.0194>, rotation=<rot -77.5566 177.3915 7.7314>}]
 clavicle_r [{shapeType=Capsule, length=5.9464, radius=10.0457, center=<-8.6486 0.4352 1.7242>, rotation=<rot -77.5565 177.3915 -172.2685>}]
 dyn_R_shoulderLeaves_jnt_1 [{
   shapeType=Capsule
   length=3.9836
   radius=3.6524
   shapeName=dyn_R_shoulderLeaves_jnt_1_capsule
   center=<2.9744 0 0.0467>
   rotation=<rot 0 -89.9999 89.1002>
  }]
 dyn_R_shoulderLeaves_jnt_2 [{
   shapeType=Capsule
   length=3.0389
   radius=3.9047
   shapeName=dyn_R_shoulderLeaves_jnt_2_capsule
   center=<2.3754 0 -0.0798>
   rotation=<rot -88.0765 -179.9995 179.9996>
  }]
 dyn_R_thighLeaf_jnt [{shapeType=Capsule, length=2.5, radius=4, shapeName=dyn_R_thighLeaf_jnt_capsule, center=<1.1623 0 -3.1933>, rotation=<rot 20.0001 0 0.0001>}]
 dyn_knot_bottom_jnt_1 [{
   shapeType=Capsule
   length=5.25
   radius=4.75
   shapeName=dyn_knot_bottom_jnt_1_capsule
   center=<-0.0522 0 -0.4235>
   rotation=<rot -80.0001 -0.0004 0>
  }]
 foot_l [{shapeType=Capsule, length=11.3482, radius=8.5127, center=<-2.2234 6.6865 -0.3433>, rotation=<rot 0.0388 1.4701 -89.3982>}]
 foot_r [{shapeType=Capsule, length=12.0982, radius=8.7627, center=<2.2072 -6.0534 0.3367>, rotation=<rot 0.0388 1.4701 90.6018>}]
 head [{shapeType=Capsule, length=0, radius=18, center=<9.0697 0.0746 0>, rotation=<rot 0 -90 0>}]
 lowerarm_l []
 lowerarm_r []
 neck_02 []
 pelvis [{shapeType=Capsule, length=3.9028, radius=16.1595, center=<-1.9544 -2.2029 -0.1981>}]
 spine_01 [{shapeType=Capsule, length=9.7671, radius=13.7354, center=<-2.8183 -3.7315 0>}]
 spine_02 [
  {shapeType=Capsule, length=9.2993, radius=15.8468, center=<1.179 -2.2934 0>}
  {shapeType=Capsule, length=20.6875, radius=9.3125, center=<-1.8317 2.606 0>, rotation=<rot 90 0 0>}
 ]
 spine_03 [
  {shapeType=Capsule, length=3.2984, radius=8.4854, center=<-0.0747 2.3725 0>}
  {shapeType=Capsule, length=20.7188, radius=9.2812, center=<-1.2326 2.6059 0>, rotation=<rot 90 0 0>}
 ]
 spine_04 [
  {shapeType=Capsule, length=6.8664, radius=15.3549, center=<3.6811 1.4112 0>}
  {shapeType=Capsule, length=10.0312, radius=9.2812, center=<-1.2326 1.308 0>, rotation=<rot 90 0 0>}
 ]
 spine_05 [{shapeType=Capsule, length=10.9859, radius=10.0791, center=<5.6377 0.3227 0>}]
 thigh_l []
 thigh_r []
 upperarm_l []
 upperarm_r []
}
constraints [
 {
  bone1Name=dyn_R_shoulderLeaves_jnt_1
  bone2Name=spine_05
  swing1LimitDegrees=45
  swing1Motion=Free
  swing2LimitDegrees=45
  swing2Motion=Free
  twistLimitDegrees=45
  twistMotion=Free
 }
 {
  bone1Name=dyn_R_shoulderLeaves_jnt_2
  bone2Name=dyn_R_shoulderLeaves_jnt_1
  swing1LimitDegrees=45
  swing1Motion=Free
  swing2LimitDegrees=45
  swing2Motion=Free
  twistLimitDegrees=45
  twistMotion=Free
 }
]
