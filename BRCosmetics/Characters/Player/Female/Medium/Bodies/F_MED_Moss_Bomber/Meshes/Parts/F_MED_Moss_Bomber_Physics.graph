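# Physics asset for the female medium Moss Bomber character part, defining capsule bodies for dynamic hair and petal bones.
# Spine, neck and head capsules are linked by free-rotation constraints, and two limited-angle chains branch from the head: a four-bone hair chain and a three-bone faceAcc hair chain.
# Four petal bodies hang off the first hair bone, each twist-locked and limited to a single swing axis, alternating between petals.
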PhysicsAsset /BRCosmetics/Characters/Player/Female/Medium/Bodies/F_MED_Moss_Bomber/Meshes/Parts/F_MED_Moss_Bomber_Physics
bodies {
 dyn_hair_02_faceAcc [{shapeType=Capsule, length=2, radius=2, center=<0.4147 0 0>, rotation=<rot 90 0 0>}]
 dyn_hair_03_faceAcc [{shapeType=Capsule, length=2, radius=3, center=<0.415 0 0>, rotation=<rot 90 0 0>}]
 dyn_hair_05 [{shapeType=Capsule, length=2.25, radius=3.5, rotation=<rot 66 -180 -180>}]
 dyn_hair_05_a [{shapeType=Capsule, length=3, radius=2.75, center=<1 0 0>, rotation=<rot 90 0 0>}]
 dyn_hair_05_b [{shapeType=Capsule, length=3.8932, radius=4.1124, rotation=<rot 79.3605 0 0>}]
 dyn_hair_05_c [{shapeType=Capsule, length=8.3018, radius=2.951, center=<1 0 0>, rotation=<rot 90 1.7899 7.872>}]
 dyn_hair_tip_faceAcc [{shapeType=Capsule, length=2, radius=2, center=<0.4147 0 0>, rotation=<rot 90 0 0>}]
 dyn_petal_jnt_1 [{shapeType=Capsule, length=2, radius=0.5, center=<1.5 0 0>, rotation=<rot 90 0 0>}]
 dyn_petal_jnt_2 [{shapeType=Capsule, length=2, radius=0.5, center=<1.5 0 0>, rotation=<rot 90 0 0>}]
 dyn_petal_jnt_3 [{shapeType=Capsule, length=2, radius=0.5, center=<1.5 0 0>, rotation=<rot 90 0 0>}]
 dyn_petal_jnt_4 [{shapeType=Capsule, length=2, radius=0.5, center=<1.5 0 0>, rotation=<rot 90 0 0>}]
 head [{shapeType=Capsule, length=7.6, radius=6.6, center=<0.1705 0.1239 -0.432>, rotation=<rot 0 80 -90>}]
 neck_01 [{shapeType=Capsule, length=4.5, radius=4.505, center=<0.414 1.1778 0>, rotation=<rot 0 -90 -90>}]
 spine_05 [{shapeType=Capsule, length=24.308, radius=11.121, center=<-0.7446 1.516 0>}]
}
constraints [
 {
  bone1Name=neck_01
  bone2Name=spine_05
  swing1LimitDegrees=45
  swing1Motion=Free
  swing2LimitDegrees=45
  swing2Motion=Free
  twistLimitDegrees=45
  twistMotion=Free
 }
 {
  bone1Name=head
  bone2Name=neck_01
  swing1LimitDegrees=45
  swing1Motion=Free
  swing2LimitDegrees=45
  swing2Motion=Free
  twistLimitDegrees=45
  twistMotion=Free
 }
 {bone1Name=dyn_hair_05, bone2Name=head, swing1LimitDegrees=20, swing2LimitDegrees=20, twistLimitDegrees=5}
 {bone1Name=dyn_hair_05_a, bone2Name=dyn_hair_05, swing1LimitDegrees=45, swing2LimitDegrees=45, twistLimitDegrees=5}
 {bone1Name=dyn_hair_05_b, bone2Name=dyn_hair_05_a, swing1LimitDegrees=45, swing2LimitDegrees=45, twistLimitDegrees=5}
 {bone1Name=dyn_hair_05_c, bone2Name=dyn_hair_05_b, swing1LimitDegrees=45, swing2LimitDegrees=45, twistLimitDegrees=5}
 {
  bone1Name=dyn_petal_jnt_1
  bone2Name=dyn_hair_05
  swing1LimitDegrees=10
  swing1Motion=Locked
  swing2LimitDegrees=15
  twistLimitDegrees=45
  twistMotion=Locked
 }
 {
  bone1Name=dyn_petal_jnt_2
  bone2Name=dyn_hair_05
  swing1LimitDegrees=15
  swing2LimitDegrees=10
  swing2Motion=Locked
  twistLimitDegrees=45
  twistMotion=Locked
 }
 {
  bone1Name=dyn_petal_jnt_3
  bone2Name=dyn_hair_05
  swing1LimitDegrees=10
  swing1Motion=Locked
  swing2LimitDegrees=15
  twistLimitDegrees=45
  twistMotion=Locked
 }
 {
  bone1Name=dyn_petal_jnt_4
  bone2Name=dyn_hair_05
  swing1LimitDegrees=15
  swing2LimitDegrees=10
  swing2Motion=Locked
  twistLimitDegrees=45
  twistMotion=Locked
 }
 {bone1Name=dyn_hair_03_faceAcc, bone2Name=head, swing1LimitDegrees=5, swing2LimitDegrees=30, twistLimitDegrees=1}
 {bone1Name=dyn_hair_02_faceAcc, bone2Name=dyn_hair_03_faceAcc, swing1LimitDegrees=20, swing2LimitDegrees=15, twistLimitDegrees=1}
 {bone1Name=dyn_hair_tip_faceAcc, bone2Name=dyn_hair_02_faceAcc, swing1LimitDegrees=3, swing2LimitDegrees=3, twistLimitDegrees=3}
]
